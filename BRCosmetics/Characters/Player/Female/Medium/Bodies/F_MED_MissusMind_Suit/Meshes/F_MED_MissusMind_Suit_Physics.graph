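# Physics asset for the F_MED_MissusMind_Suit female body cosmetic, set up for secondary motion on the suit's coattail.
# A chain of three coattail capsules hangs from a tiny master capsule, with twist locked and narrow side swing, while the primary swing limit widens toward the tip.
# Unconstrained pelvis and thigh capsules provide body volumes for the coattail to collide against.
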PhysicsAsset /BRCosmetics/Characters/Player/Female/Medium/Bodies/F_MED_MissusMind_Suit/Meshes/F_MED_MissusMind_Suit_Physics
bodies {
 dyn_coattail_01 [{shapeType=Capsule, length=25.0552, radius=3.2514, shapeName=dyn_coattail_01_capsule, center=<4.0715 0 0>}]
 dyn_coattail_02 [{shapeType=Capsule, length=20.2753, radius=3.1296, shapeName=dyn_coattail_02_capsule, center=<4.0715 0 0>}]
 dyn_coattail_03 [{shapeType=Capsule, length=8.8285, radius=2.56, shapeName=dyn_coattail_03_capsule, center=<4.0715 0 0>}]
 dyn_coattail_mstr [{shapeType=Capsule, length=0.505, radius=0.505, shapeName=dyn_coattail_mstr_capsule, rotation=<rot 0 -90 0>}]
 pelvis [{shapeType=Capsule, length=8, radius=9.7397, shapeName=pelvis_capsule}]
 thigh_l [{shapeType=Capsule, length=31.451, radius=7.6368, shapeName=thigh_l_capsule, center=<-18.9991 0 1.4801>, rotation=<rot 90 0 0>}]
 thigh_r [{shapeType=Capsule, length=31.451, radius=7.6368, shapeName=thigh_r_capsule, center=<18.9991 0 -1.4801>, rotation=<rot -90 180 180>}]
}
constraints [
 {bone1Name=dyn_coattail_01, bone2Name=dyn_coattail_mstr, swing1LimitDegrees=12, swing2LimitDegrees=6, twistLimitDegrees=0, twistMotion=Locked}
 {bone1Name=dyn_coattail_02, bone2Name=dyn_coattail_01, swing1LimitDegrees=30, swing2LimitDegrees=6, twistLimitDegrees=0, twistMotion=Locked}
 {bone1Name=dyn_coattail_03, bone2Name=dyn_coattail_02, swing1LimitDegrees=35, swing2LimitDegrees=6, twistLimitDegrees=0, twistMotion=Locked}
]
