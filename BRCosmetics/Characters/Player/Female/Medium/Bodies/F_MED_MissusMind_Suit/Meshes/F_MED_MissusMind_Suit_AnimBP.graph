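# Anim Blueprint for the female MissusMind suit cosmetic body, adding secondary motion to a pose copied from the parent mesh.
# The pose runs through a Control Rig, pelvis-space RigidBody physics and an additive coattail bone modifier with zero rotation.
# Each update sets the gravity override via a deprecated state function and lowers physics alpha while skydiving.
# It also computes a sine-based wind rotator at skydive-dependent frequencies, which the shown graphs never read.
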AnimBlueprint /BRCosmetics/Characters/Player/Female/Medium/Bodies/F_MED_MissusMind_Suit/Meshes/F_MED_MissusMind_Suit_AnimBP
target_skeleton=&/F_MED_MissusMind_Suit_Skeleton
graphs [
 AnimGraph {nodes [
   n1 = Control Rig(Source=Copy Pose From Mesh() {bUseAttachedParent=true}) {
    alphaBoolBlend {blendOption=Linear}
    controlRigAssetReference {blueprintRigClass=&/F_MED_MissusMind_Suit_CtrlRig.F_MED_MissusMind_Suit_CtrlRig_C}
   }
   n2 = RigidBody!(
    ComponentPose=Local To Component(LocalPose=@n1)
    Alpha=Get RBAN_Alpha() -> (RBAN_Alpha:double)
    OverrideWorldGravity:Vector=Get Gravity Override()
    ExternalForce
   ) -> (Pose) {
    bClampLinearTranslationLimitToRefPose=true
    bOverrideWorldGravity=true
    overridePhysicsAsset=&/F_MED_MissusMind_Suit_Physics
    simulationSpace=BaseBoneSpace
    alphaBoolBlend {blendOption=Linear}
    baseBoneRef {boneName=pelvis}
    componentAppliedLinearAccClamp=<5000 5000 5000>
    componentLinearAccScale=<0.4 0.4 0.4>
    componentLinearVelScale=<0.5 0.5 0.5>
    simSpaceSettings {maxAngularAcceleration=100, maxAngularVelocity=300, maxLinearAcceleration=300, maxLinearVelocity=500, worldAlpha=0.7}
   }
   n3 = "Transform (Modify) Bone - Bone: dyn_coattail_02"!(ComponentPose=@n2, bAlphaBoolEnabled:bool=True, Rotation="0,0,0") -> (Pose) {
    alphaInputType=Bool
    rotationMode=BMM_Additive
    rotationSpace=BCS_BoneSpace
    alphaBoolBlend {blendOption=Linear}
    boneToModify {boneName=dyn_coattail_02}
   }
   Output Pose(Result=Component To Local(ComponentPose=@n3)) {layerGroup=None}
  ]}
 EventGraph {nodes [
   n1 = Event BlueprintUpdateAnimation()
   > n2 = Sequence!(execute) -> (then_0, then_1, then_2:Exec, then_3:Exec, then_4:Exec)
   > n3 = Gravity Override State Function DEPRECATED(
    CharacterPartAnimInstance=Self-Reference()
    GameTimeInSeconds=@n1.DeltaTimeX
    GravityOverrideParams=Get Gravity Override Params()
    RBAN Params=Get RBAN Params()
   )
   > Set Gravity Override(Gravity Override=@n3)
   n4 = Branch(execute=@n2.then_1, Condition=Get bIsSkyDiving())
   > Set RBAN_Alpha(RBAN_Alpha:double=0.5) -> (Output_Get:double)
   > Set EnableWind(EnableWind:bool=true) -> (Output_Get:bool)
   > n5 = Branch(Condition=Get bIsSkydiveDiveMode())
   > Set WindFrequency(WindFrequency:double=35) -> (Output_Get:double)
   n6 = Branch(execute=@n5.else, Condition=Get bIsSkydiveFloating())
   > Set WindFrequency(WindFrequency=30)
   Set WindFrequency(execute=@n6.else, WindFrequency=27)
   Set RBAN_Alpha(execute=@n4.else, RBAN_Alpha=1)
   > Set EnableWind(EnableWind=false)
   Set CoattailRotator(
    execute=@n2.then_4
    CoattailRotator:Rotator=Rotator Rumble!(frequency:double=Get WindFrequency() -> (WindFrequency:double), amplitude:double=2, time offset:Vector="0,1,2", coefficient:Vector="0,1,0") -> (rotatorOutput:Rotator=~) {is=title K2Node_MacroInstance}
   ) -> (Output_Get:Rotator)
  ]}
 floatSinFunc {nodes [
   n1 = Inputs!() -> (time:double=~, time offset:double=~, frequency:double=~, amplitude:double=~, coefficient:double=~)
   n2 = float + float(A=@n1, B=@n1.time offset)
   n3 = MultiplyByPi!(Value:double="Sin (Radians)"(A=float * float(A=@n2, B=@n1.frequency))) -> (ReturnValue:double)
   Outputs!(floatOutput:double=float * float(A=float * float(A=@n3, B=@n1.amplitude), B=@n1.coefficient))
  ]}
 rotatorRumble {nodes [
   n1 = Inputs!() -> (frequency, amplitude, time offset:Vector=~, coefficient:Vector=~)
   n2 = BreakVector(InVec=@n1.coefficient)
   n3 = BreakVector(InVec=@n1.time offset)
   n4 = GetGameTimeInSeconds()
   Outputs!(rotatorOutput:Rotator=MakeRotator(
     Roll=Float Sin Func!(time:double=@n4, time offset:double=@n3, frequency:double=@n1, amplitude:double=@n1.amplitude, coefficient:double=@n2) -> (floatOutput:double=~) {is=title K2Node_MacroInstance}
     Pitch=Float Sin Func(time=@n4, time offset=@n3.Z, frequency=@n1, amplitude=@n1.amplitude, coefficient=@n2.Z) -> (floatOutput=~)
     Yaw=Float Sin Func(time=@n4, time offset=@n3.Y, frequency=@n1, amplitude=@n1.amplitude, coefficient=@n2.Y) -> (floatOutput=~)
    ))
  ]}
]
variables [
 CoattailRotator {type=Rotator}
 EnableWind {type=boolean}
 Gravity Override {type=Vector}
 Gravity Override Params {type=GravityOverrideParamsStruct}
 RBAN Params {
  type=RBANParamsStruct
  default {
   dBNO {simSpaceSettings {
     maxAngularAcceleration=1000
     maxAngularVelocity=3
     maxLinearAcceleration=300
     maxLinearVelocity=700
     velocityScaleZ=0
     externalLinearDragV=<0 0.5 0>
    }}
   dBNO_Carried {simSpaceSettings {
     maxAngularAcceleration=1000
     maxAngularVelocity=3
     maxLinearAcceleration=300
     maxLinearVelocity=700
     velocityScaleZ=0
     externalLinearDragV=<0 0.5 0>
    }}
   emoteOrMelee {simSpaceSettings {
     maxAngularAcceleration=1000
     maxAngularVelocity=3
     maxLinearAcceleration=300
     maxLinearVelocity=700
     velocityScaleZ=0
     externalLinearDragV=<0 0.5 0>
    }}
   front_End {simSpaceSettings {
     maxAngularAcceleration=1000
     maxAngularVelocity=3
     maxLinearAcceleration=300
     maxLinearVelocity=700
     velocityScaleZ=0
     externalLinearDragV=<0 0.5 0>
    }}
   onGround_Crouch_Moving {simSpaceSettings {
     maxAngularAcceleration=1000
     maxAngularVelocity=3
     maxLinearAcceleration=300
     maxLinearVelocity=700
     velocityScaleZ=0
     externalLinearDragV=<0 0.5 0>
    }}
   onGround_Falling {simSpaceSettings {
     maxAngularAcceleration=1000
     maxAngularVelocity=3
     maxLinearAcceleration=300
     maxLinearVelocity=700
     velocityScaleZ=0
     externalLinearDragV=<0 0.5 0>
    }}
   onGround_Moving {simSpaceSettings {
     maxAngularAcceleration=1000
     maxAngularVelocity=3
     maxLinearAcceleration=300
     maxLinearVelocity=700
     velocityScaleZ=0
     externalLinearDragV=<0 0.5 0>
    }}
   onGround_Standing {simSpaceSettings {
     maxAngularAcceleration=1000
     maxAngularVelocity=3
     maxLinearAcceleration=300
     maxLinearVelocity=700
     velocityScaleZ=0
     externalLinearDragV=<0 0.5 0>
    }}
   rail_Grinding {simSpaceSettings {
     maxAngularAcceleration=1000
     maxAngularVelocity=3
     maxLinearAcceleration=300
     maxLinearVelocity=700
     velocityScaleZ=0
     worldAlpha=0.1
     externalLinearDragV=<0 0.5 0>
    }}
   riding_Moving {simSpaceSettings {
     maxAngularAcceleration=1000
     maxAngularVelocity=3
     maxLinearAcceleration=300
     maxLinearVelocity=700
     velocityScaleZ=0
     externalLinearDragV=<0 0.5 0>
    }}
   riding_Moving_Targeting {simSpaceSettings {
     maxAngularAcceleration=1000
     maxAngularVelocity=3
     maxLinearAcceleration=300
     maxLinearVelocity=700
     velocityScaleZ=0
     externalLinearDragV=<0 0.5 0>
    }}
   riding_Standing {simSpaceSettings {
     maxAngularAcceleration=1000
     maxAngularVelocity=3
     maxLinearAcceleration=300
     maxLinearVelocity=700
     velocityScaleZ=0
     externalLinearDragV=<0 0.5 0>
    }}
   skydive_DivingDown {simSpaceSettings {
     maxAngularAcceleration=1000
     maxAngularVelocity=3
     maxLinearAcceleration=300
     maxLinearVelocity=700
     velocityScaleZ=0
     externalLinearDragV=<0 0.5 0>
    }}
   skydive_FloatingDown {
    componentLinearAccScale=<0 0.5 0>
    componentLinearVelScale=<0 0.5 0>
    simSpaceSettings {
     maxAngularAcceleration=1000
     maxAngularVelocity=3
     maxLinearAcceleration=300
     maxLinearVelocity=700
     velocityScaleZ=0
     externalLinearDragV=<0 0.5 0>
    }
   }
   skydive_GoingUp {simSpaceSettings {
     maxAngularAcceleration=1000
     maxAngularVelocity=3
     maxLinearAcceleration=300
     maxLinearVelocity=700
     velocityScaleZ=0
     externalLinearDragV=<0 0.5 0>
    }}
   skydive_Parachuting {
    componentLinearAccScale=<0 0.1 0.3>
    componentLinearVelScale=<0 0.1 0.3>
    simSpaceSettings {
     maxAngularAcceleration=1000
     maxAngularVelocity=3
     maxLinearAcceleration=300
     maxLinearVelocity=700
     velocityScaleZ=0
     externalLinearDragV=<0 0.5 0>
    }
   }
   skydive_Surfing {simSpaceSettings {
     maxAngularAcceleration=1000
     maxAngularVelocity=3
     maxLinearAcceleration=300
     maxLinearVelocity=700
     velocityScaleZ=0
     externalLinearDragV=<0 0.5 0>
    }}
   swimming {simSpaceSettings {
     maxAngularAcceleration=1000
     maxAngularVelocity=3
     maxLinearAcceleration=300
     maxLinearVelocity=700
     velocityScaleZ=0
     externalLinearDragV=<0 0.5 0>
    }}
   tactical_Sprint {simSpaceSettings {
     maxAngularAcceleration=1000
     maxAngularVelocity=3
     maxLinearAcceleration=300
     maxLinearVelocity=700
     velocityScaleZ=0
     externalLinearDragV=<0 0.5 0>
    }}
   ziplining {simSpaceSettings {
     maxAngularAcceleration=1000
     maxAngularVelocity=3
     maxLinearAcceleration=300
     maxLinearVelocity=700
     velocityScaleZ=0
     externalLinearDragV=<0 0.5 0>
    }}
  }
 }
 RBAN_Alpha {type=number}
 WindFrequency {type=number}
]
