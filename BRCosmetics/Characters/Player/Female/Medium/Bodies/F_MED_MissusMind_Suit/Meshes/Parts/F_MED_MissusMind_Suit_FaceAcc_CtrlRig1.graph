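# Control Rig for the MissusMind female suit's face accessory that poses three long hair chains (left, center, right) against the body.
# Forward solve parent-constrains control nulls to weighted body bones, then fits each chain through them with B-spline IK.
# Hair Adjustments then adds corrective offsets from spherical pose readers on head (outside front end), upper arms and right clavicle (with pack), plus a skydiving offset.
# Construction caches control lists and poses.
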ControlRigBlueprint /BRCosmetics/Characters/Player/Female/Medium/Bodies/F_MED_MissusMind_Suit/Meshes/Parts/F_MED_MissusMind_Suit_FaceAcc_CtrlRig1
graphs [
 New Function_ContainedGraph {nodes [
   n1 = RigVMFunctionEntryNode()
   RigVMFunction_ControlFlowBranch = &RigVMFunction_ControlFlowBranch(ExecuteContext=@n1.ExecuteContext, Condition=@n1.debug)
   RigVMFunctionReturnNode(ExecuteContext=@RigVMFunction_ControlFlowBranch.Completed)
   n2 = &RigUnit_ItemArray(Items="((Type=Bone,Name=\"spine_01\"),(Type=Bone,Name=\"spine_02\"),(Type=Bone,Name=\"spine_03\"),(Type=Bone,Name=\"spine_04\"),(Type=Bone,Name=\"spine_05\"),(Type=Bone,Name=\"neck_01\"),(Type=Bone,Name=\"neck_02\"),(Type=Bone,Name=\"head\"),(Type=Bone,Name=\"clavicle_l\"),(Type=B ...(truncated)")
   For_Each = RigVMDispatchNode(Array=@n2.Items, ExecuteContext=@RigVMFunction_ControlFlowBranch.True) {template="DISPATCH_RigVMDispatch_ArrayIterator(in Array,out Element,out Index,out Count,out Ratio)"}
   Concat = /Script/RigVM.RigVMFunction_NameConcat(B=None, A=@For_Each.Element.Name) {is=struct RigVMUnitNode, template="Concat::Execute(in A,in B,out Result)"}
   > &RigUnit_SetTransform(
    Item=(Type=Bone,Name="None")
    Weight=1
    bInitial=False
    Item.Name=@For_Each.Element.Name
    Value=&RigUnit_GetTransform(Item=(Type=Control,Name="None"), bInitial=False, Item.Name=@Concat.Result).Transform
   )
   &RigUnit_SetTransform(
    Item=(Type=Bone,Name="pelvis")
    Weight=1
    bInitial=False
    bPropagateToChildren=false
    ExecutePin=@For_Each.Completed
    Value=&RigUnit_GetTransform(Item=(Type=Control,Name="pelvis_ctrl"), bInitial=False).Transform
   )
  ]}
 New Function_ContainedGraph {nodes [
   RigVMFunctionEntryNode()
   > RigVMFunction_Sequence = RigVMAggregateNode()
   &RigUnit_ParentConstraint(
    AdvancedSettings=(RotationOrderForFilter=XZY)
    Child=(Type=Null,Name="dyn_c_hair_01_null")
    Parents=((Item=(Type=Bone,Name="neck_01"),Weight=0.0758),(Item=(Type=Bone,Name="spine_05"),Weight=0.0952),(Item=(Type=Bone,Name="neck_02"),Weight=0.384),(Item=(Type=Bone,Name="head"),Weight=0.445))
    Weight=1
    ExecutePin=@RigVMFunction_Sequence.A
   )
   > &RigUnit_ParentConstraint(
    AdvancedSettings=(RotationOrderForFilter=XZY)
    Child=(Type=Null,Name="dyn_c_hair_02_null")
    Parents=((Item=(Type=Bone,Name="neck_01"),Weight=0.142),(Item=(Type=Bone,Name="spine_05"),Weight=0.2186),(Item=(Type=Bone,Name="head"),Weight=0.2286),(Item=(Type=Bone,Name="neck_02"),Weight=0.4108))
    Weight=1
   )
   > &RigUnit_ParentConstraint(
    AdvancedSettings=(RotationOrderForFilter=XZY)
    Child=(Type=Null,Name="dyn_c_hair_03_null")
    Parents="((Item=(Type=Bone,Name=\"neck_02\"),Weight=0.0143),(Item=(Type=Bone,Name=\"clavicle_r\"),Weight=0.0426),(Item=(Type=Bone,Name=\"neck_01\"),Weight=0.0438),(Item=(Type=Bone,Name=\"clavicle_l\"),Weight=0.0467),(Item=(Type=Bone,Name=\"spine_03\"),Weight=0.0479),(Item=(T ...(truncated)"
    Weight=1
   )
   > &RigUnit_ParentConstraint(
    AdvancedSettings=(RotationOrderForFilter=XZY)
    Child=(Type=Null,Name="dyn_c_hair_04_null")
    Parents=((Item=(Type=Bone,Name="spine_05"),Weight=1.0))
    Weight=1
   )
   > &RigUnit_ParentConstraint(
    AdvancedSettings=(RotationOrderForFilter=XZY)
    Child=(Type=Null,Name="dyn_c_hair_05_null")
    Parents=((Item=(Type=Bone,Name="spine_05"),Weight=1.0))
    Weight=1
   )
   > &RigUnit_ParentConstraint(
    AdvancedSettings=(RotationOrderForFilter=XZY)
    Child=(Type=Null,Name="dyn_c_hair_06_null")
    Parents=((Item=(Type=Bone,Name="spine_05"),Weight=1.0))
    Weight=1
   )
   > &RigUnit_ParentConstraint(
    AdvancedSettings=(RotationOrderForFilter=XZY)
    Child=(Type=Null,Name="dyn_c_hair_07_null")
    Parents="((Item=(Type=Bone,Name=\"pelvis\"),Weight=0.0165),(Item=(Type=Bone,Name=\"clavicle_l\"),Weight=0.0507),(Item=(Type=Bone,Name=\"clavicle_r\"),Weight=0.0589),(Item=(Type=Bone,Name=\"spine_01\"),Weight=0.0597),(Item=(Type=Bone,Name=\"spine_05\"),Weight=0.1317),(Item=(T ...(truncated)"
    Weight=1
   )
   > &RigUnit_ParentConstraint(
    AdvancedSettings=(RotationOrderForFilter=XZY)
    Child=(Type=Null,Name="dyn_l_hair_01_null")
    Parents=((Item=(Type=Bone,Name="neck_01"),Weight=0.0922),(Item=(Type=Bone,Name="spine_05"),Weight=0.124),(Item=(Type=Bone,Name="neck_02"),Weight=0.3759),(Item=(Type=Bone,Name="head"),Weight=0.4079))
    Weight=1
   )
   > &RigUnit_ParentConstraint(
    AdvancedSettings=(RotationOrderForFilter=XZY)
    Child=(Type=Null,Name="dyn_l_hair_02_null")
    Parents=((Item=(Type=Bone,Name="neck_01"),Weight=0.1571),(Item=(Type=Bone,Name="head"),Weight=0.1722),(Item=(Type=Bone,Name="spine_05"),Weight=0.3006),(Item=(Type=Bone,Name="neck_02"),Weight=0.3702))
    Weight=1
   )
   > &RigUnit_ParentConstraint(
    AdvancedSettings=(RotationOrderForFilter=XZY)
    Child=(Type=Null,Name="dyn_l_hair_03_null")
    Parents=((Item=(Type=Bone,Name="neck_01"),Weight=0.0052),(Item=(Type=Bone,Name="spine_03"),Weight=0.0057),(Item=(Type=Bone,Name="clavicle_l"),Weight=0.4754),(Item=(Type=Bone,Name="spine_05"),Weight=0.5138))
    Weight=1
   )
   > &RigUnit_ParentConstraint(
    AdvancedSettings=(RotationOrderForFilter=XZY)
    Child=(Type=Null,Name="dyn_l_hair_04_null")
    Parents=((Item=(Type=Bone,Name="spine_05"),Weight=1.0))
    Weight=1
   )
   > &RigUnit_ParentConstraint(
    AdvancedSettings=(RotationOrderForFilter=XZY)
    Child=(Type=Null,Name="dyn_l_hair_05_null")
    Parents=((Item=(Type=Bone,Name="spine_05"),Weight=1.0))
    Weight=1
   )
   > &RigUnit_ParentConstraint(
    AdvancedSettings=(RotationOrderForFilter=XZY)
    Child=(Type=Null,Name="dyn_l_hair_06_null")
    Parents="((Item=(Type=Bone,Name=\"spine_01\"),Weight=0.0184),(Item=(Type=Bone,Name=\"clavicle_r\"),Weight=0.028),(Item=(Type=Bone,Name=\"upperarm_l\"),Weight=0.0373),(Item=(Type=Bone,Name=\"spine_03\"),Weight=0.0657),(Item=(Type=Bone,Name=\"spine_02\"),Weight=0.0729),(Item=( ...(truncated)"
    Weight=1
   )
   > &RigUnit_ParentConstraint(
    AdvancedSettings=(RotationOrderForFilter=XZY)
    Child=(Type=Null,Name="dyn_l_hair_07_null")
    Parents="((Item=(Type=Bone,Name=\"pelvis\"),Weight=0.0158),(Item=(Type=Bone,Name=\"clavicle_r\"),Weight=0.0186),(Item=(Type=Bone,Name=\"spine_01\"),Weight=0.0556),(Item=(Type=Bone,Name=\"clavicle_l\"),Weight=0.1143),(Item=(Type=Bone,Name=\"spine_05\"),Weight=0.1239),(Item=(T ...(truncated)"
    Weight=1
   )
   > &RigUnit_ParentConstraint(
    AdvancedSettings=(RotationOrderForFilter=XZY)
    Child=(Type=Null,Name="dyn_r_hair_01_null")
    Parents=((Item=(Type=Bone,Name="neck_01"),Weight=0.0919),(Item=(Type=Bone,Name="spine_05"),Weight=0.1241),(Item=(Type=Bone,Name="neck_02"),Weight=0.3752),(Item=(Type=Bone,Name="head"),Weight=0.4088))
    Weight=1
   )
   > &RigUnit_ParentConstraint(
    AdvancedSettings=(RotationOrderForFilter=XZY)
    Child=(Type=Null,Name="dyn_r_hair_02_null")
    Parents=((Item=(Type=Bone,Name="neck_01"),Weight=0.1568),(Item=(Type=Bone,Name="head"),Weight=0.1727),(Item=(Type=Bone,Name="spine_05"),Weight=0.3008),(Item=(Type=Bone,Name="neck_02"),Weight=0.3697))
    Weight=1
   )
   > &RigUnit_ParentConstraint(
    AdvancedSettings=(RotationOrderForFilter=XZY)
    Child=(Type=Null,Name="dyn_r_hair_03_null")
    Parents=((Item=(Type=Bone,Name="neck_01"),Weight=0.0052),(Item=(Type=Bone,Name="spine_03"),Weight=0.0056),(Item=(Type=Bone,Name="clavicle_r"),Weight=0.4748),(Item=(Type=Bone,Name="spine_05"),Weight=0.5144))
    Weight=1
   )
   > &RigUnit_ParentConstraint(
    AdvancedSettings=(RotationOrderForFilter=XZY)
    Child=(Type=Null,Name="dyn_r_hair_04_null")
    Parents=((Item=(Type=Bone,Name="spine_05"),Weight=1.0))
    Weight=1
   )
   > &RigUnit_ParentConstraint(
    AdvancedSettings=(RotationOrderForFilter=XZY)
    Child=(Type=Null,Name="dyn_r_hair_05_null")
    Parents=((Item=(Type=Bone,Name="spine_05"),Weight=1.0))
    Weight=1
   )
   > &RigUnit_ParentConstraint(
    AdvancedSettings=(RotationOrderForFilter=XZY)
    Child=(Type=Null,Name="dyn_r_hair_06_null")
    Parents="((Item=(Type=Bone,Name=\"spine_01\"),Weight=0.0184),(Item=(Type=Bone,Name=\"clavicle_l\"),Weight=0.028),(Item=(Type=Bone,Name=\"upperarm_r\"),Weight=0.0373),(Item=(Type=Bone,Name=\"spine_03\"),Weight=0.0657),(Item=(Type=Bone,Name=\"spine_02\"),Weight=0.0728),(Item=( ...(truncated)"
    Weight=1
   )
   > &RigUnit_ParentConstraint(
    AdvancedSettings=(RotationOrderForFilter=XZY)
    Child=(Type=Null,Name="dyn_r_hair_07_null")
    Parents="((Item=(Type=Bone,Name=\"pelvis\"),Weight=0.0155),(Item=(Type=Bone,Name=\"clavicle_l\"),Weight=0.0188),(Item=(Type=Bone,Name=\"spine_01\"),Weight=0.0556),(Item=(Type=Bone,Name=\"clavicle_r\"),Weight=0.114),(Item=(Type=Bone,Name=\"spine_05\"),Weight=0.124),(Item=(Typ ...(truncated)"
    Weight=1
   )
   n1 = &RigUnit_ItemArray(Items=((Type=Bone,Name="dyn_c_hair_01"),(Type=Bone,Name="dyn_c_hair_02"),(Type=Bone,Name="dyn_c_hair_03"),(Type=Bone,Name="dyn_c_hair_04"),(Type=Bone,Name="dyn_c_hair_05"),(Type=Bone,Name="dyn_c_hair_06"),(Type=Bone,Name="dyn_c_hair_07")))
   SplineIK = RigVMFunctionReferenceNode(
    Primary Axis=(X=1,Y=0,Z=0)
    Samples Per Segment=4
    Secondary Spline Direction=(X=0,Y=5,Z=0)
    Spline Mode=BSpline
    Stretch=false
    Up Axis=(X=0,Y=0,Z=0)
    debug=false
    Bones=@n1.Items
    ExecuteContext=@RigVMFunction_Sequence.B
    Controls=RigVMVariableNode() {variable=c_dyn_hair_controls}.Value
   )
   RigUnit_ItemArray = &RigUnit_ItemArray(Items="((Type=Control,Name=\"dyn_l_hair_01_ctrl\"),(Type=Control,Name=\"dyn_l_hair_02_ctrl\"),(Type=Control,Name=\"dyn_l_hair_03_ctrl\"),(Type=Control,Name=\"dyn_l_hair_04_ctrl\"),(Type=Control,Name=\"dyn_l_hair_05_ctrl\"),(Type=Control,Name=\"dyn_l_hair_06_ctrl\"),(Type=Con ...(truncated)")
   n2 = &RigUnit_ItemArray(Items=((Type=Bone,Name="dyn_l_hair_01"),(Type=Bone,Name="dyn_l_hair_02"),(Type=Bone,Name="dyn_l_hair_03"),(Type=Bone,Name="dyn_l_hair_04"),(Type=Bone,Name="dyn_l_hair_05"),(Type=Bone,Name="dyn_l_hair_06"),(Type=Bone,Name="dyn_l_hair_07")))
   > SplineIK_0 = RigVMFunctionReferenceNode(
    Primary Axis=(X=1,Y=0,Z=0)
    Samples Per Segment=4
    Secondary Spline Direction=(X=0,Y=5,Z=0)
    Spline Mode=BSpline
    Stretch=false
    Up Axis=(X=0,Y=1,Z=0)
    debug=false
    Controls=@RigUnit_ItemArray.Items
    Bones=@n2.Items
   )
   n3 = &RigUnit_ItemArray(Items=((Type=Bone,Name="dyn_r_hair_01"),(Type=Bone,Name="dyn_r_hair_02"),(Type=Bone,Name="dyn_r_hair_03"),(Type=Bone,Name="dyn_r_hair_04"),(Type=Bone,Name="dyn_r_hair_05"),(Type=Bone,Name="dyn_r_hair_06"),(Type=Bone,Name="dyn_r_hair_07")))
   > SplineIK_1_1 = RigVMFunctionReferenceNode(
    Primary Axis=(X=1,Y=0,Z=0)
    Samples Per Segment=4
    Secondary Spline Direction=(X=0,Y=5,Z=0)
    Spline Mode=BSpline
    Stretch=false
    Up Axis=(X=0,Y=0,Z=0)
    debug=false
    Bones=@n3.Items
    Controls=RigVMVariableNode() {variable=r_dyn_hair_controls}.Value
   )
   RigVMFunctionReturnNode(ExecuteContext=@RigVMFunction_Sequence.C)
  ]}
 New Function_ContainedGraph {nodes [
   RigVMFunctionEntryNode()
   > RigVMFunction_Sequence = RigVMAggregateNode()
   n1 = &RigVMFunction_ControlFlowBranch(ExecuteContext=@RigVMFunction_Sequence.A, Condition=RigVMVariableNode() {variable=isFrontEnd}.Value)
   n2 = &RigUnit_SphericalPoseReader(
    ActiveRegionScaleFactors=(PositiveWidth=1,NegativeWidth=1,PositiveHeight=1,NegativeHeight=1)
    ActiveRegionSize=0.15
    Debug=(bDrawDebug=false,bDraw2D=False,bDrawLocalAxes=False,DebugScale=25,DebugSegments=20,DebugThickness=0.25)
    DriverAxis=(X=0,Y=1,Z=0)
    DriverItem=(Type=Bone,Name="head")
    FalloffRegionScaleFactors=(PositiveWidth=1,NegativeWidth=1,PositiveHeight=1,NegativeHeight=1)
    FalloffSize=0.4
    FlipHeightScaling=False
    FlipWidthScaling=False
    OptionalParentItem=(Type=Bone,Name="spine_05")
    RotationOffset=(X=-200,Y=0,Z=0)
    ExecutePin=@n1.False
   )
   RigVMFunction_AlphaInterp = &RigVMFunction_AlphaInterp(InterpSpeedDecreasing=40, InterpSpeedIncreasing=40, Scale=1, bClampResult=False, bInterpResult=True, Value=@n2.OutputParam)
   > &RigUnit_ModifyTransforms(
    ItemToModify="((Item=(Type=Bone,Name=\"dyn_l_hair_mstr\"),Transform=(Rotation=(X=0,Y=0,Z=0,W=1),Translation=(X=0,Y=-1,Z=0),Scale3D=(X=1,Y=1,Z=1))),(Item=(Type=Bone,Name=\"dyn_c_hair_mstr\"),Transform=(Rotation=(X=0,Y=0,Z=0,W=1),Translation=(X=0,Y=-1.5,Z=0),Scale3D=(X=1,Y=1, ...(truncated)"
    WeightMinimum=0
    Weight=@RigVMFunction_AlphaInterp.Result
   )
   > n3 = &RigUnit_SphericalPoseReader(
    ActiveRegionScaleFactors=(PositiveWidth=1,NegativeWidth=1,PositiveHeight=1,NegativeHeight=1)
    ActiveRegionSize=0.15
    Debug=(bDrawDebug=false,bDraw2D=False,bDrawLocalAxes=False,DebugScale=25,DebugSegments=20,DebugThickness=0.25)
    DriverAxis=(X=0,Y=1,Z=0)
    DriverItem=(Type=Bone,Name="head")
    FalloffRegionScaleFactors=(PositiveWidth=1,NegativeWidth=1,PositiveHeight=1,NegativeHeight=1)
    FalloffSize=0.3
    FlipHeightScaling=False
    FlipWidthScaling=False
    OptionalParentItem=(Type=Bone,Name="spine_05")
    RotationOffset=(X=45,Y=0,Z=0)
   )
   n4 = &RigVMFunction_AlphaInterp(InterpSpeedDecreasing=40, InterpSpeedIncreasing=40, Scale=1, bClampResult=False, bInterpResult=True, Value=@n3.OutputParam)
   > &RigUnit_ModifyTransforms(
    ItemToModify="((Item=(Type=Bone,Name=\"dyn_l_hair_mstr\"),Transform=(Rotation=(X=0,Y=0,Z=0,W=1),Translation=(X=0,Y=-1,Z=0),Scale3D=(X=1,Y=1,Z=1))),(Item=(Type=Bone,Name=\"dyn_c_hair_mstr\"),Transform=(Rotation=(X=0,Y=0,Z=0,W=1),Translation=(X=0,Y=-1,Z=0),Scale3D=(X=1,Y=1,Z= ...(truncated)"
    WeightMinimum=0
    Weight=@n4.Result
   )
   n5 = &RigUnit_SphericalPoseReader(
    ActiveRegionScaleFactors=(PositiveWidth=1,NegativeWidth=1,PositiveHeight=1,NegativeHeight=1)
    ActiveRegionSize=0.15
    Debug=(bDrawDebug=false,bDraw2D=False,bDrawLocalAxes=False,DebugScale=25,DebugSegments=20,DebugThickness=0.25)
    DriverAxis=(X=1,Y=0,Z=0)
    DriverItem=(Type=Bone,Name="upperarm_l")
    FalloffRegionScaleFactors=(PositiveWidth=1,NegativeWidth=1,PositiveHeight=1,NegativeHeight=1)
    FalloffSize=0.4
    FlipHeightScaling=False
    FlipWidthScaling=False
    OptionalParentItem=(Type=Bone,Name="None")
    RotationOffset=(X=17.2691,Y=-12.0456,Z=28.5116)
    ExecutePin=@RigVMFunction_Sequence.B
   )
   > &RigUnit_ModifyTransforms(
    ItemToModify="((Item=(Type=Bone,Name=\"dyn_l_hair_mstr\"),Transform=(Rotation=(X=0,Y=0.0523,Z=0,W=0.9986),Translation=(X=0,Y=-1,Z=0),Scale3D=(X=1,Y=1,Z=1))),(Item=(Type=Bone,Name=\"dyn_l_hair_02\"),Transform=(Rotation=(X=0,Y=0,Z=0,W=1),Translation=(X=0,Y=-2,Z=-0.5),Scale3D= ...(truncated)"
    WeightMinimum=0
    Weight=@n5.OutputParam
   )
   > n6 = &RigUnit_SphericalPoseReader(
    ActiveRegionScaleFactors=(PositiveWidth=1,NegativeWidth=1,PositiveHeight=1,NegativeHeight=1)
    ActiveRegionSize=0.1
    Debug=(bDrawDebug=false,bDraw2D=False,bDrawLocalAxes=False,DebugScale=25,DebugSegments=20,DebugThickness=0.25)
    DriverAxis=(X=-1,Y=0,Z=0)
    DriverItem=(Type=Bone,Name="upperarm_r")
    FalloffRegionScaleFactors=(PositiveWidth=1,NegativeWidth=1,PositiveHeight=1,NegativeHeight=1)
    FalloffSize=0.5
    FlipHeightScaling=False
    FlipWidthScaling=False
    OptionalParentItem=(Type=Bone,Name="None")
    RotationOffset=(X=-114.8296,Y=5.0511,Z=35.1428)
   )
   > &RigUnit_ModifyTransforms(
    ItemToModify="((Item=(Type=Bone,Name=\"dyn_r_hair_mstr\"),Transform=(Rotation=(X=0,Y=0,Z=-0.0349,W=0.9994),Translation=(X=0,Y=-1,Z=0),Scale3D=(X=1,Y=1,Z=1))),(Item=(Type=Bone,Name=\"dyn_c_hair_mstr\"),Transform=(Rotation=(X=0,Y=0.0175,Z=0,W=0.9998),Translation=(X=0,Y=0,Z=0) ...(truncated)"
    WeightMinimum=0
    Weight=@n6.OutputParam
   )
   RigVMFunction_ControlFlowBranch = &RigVMFunction_ControlFlowBranch(ExecuteContext=@RigVMFunction_Sequence.C, Condition=RigVMVariableNode() {variable=isSkydiving}.Value)
   &RigUnit_ModifyTransforms(
    ItemToModify="((Item=(Type=Bone,Name=\"dyn_l_hair_02\"),Transform=(Rotation=(X=0,Y=0,Z=0,W=1),Translation=(X=0,Y=0,Z=1),Scale3D=(X=1,Y=1,Z=1))),(Item=(Type=Bone,Name=\"dyn_r_hair_02\"),Transform=(Rotation=(X=0,Y=0,Z=0,W=1),Translation=(X=0,Y=0,Z=0),Scale3D=(X=1,Y=1,Z=1))),( ...(truncated)"
    Weight=1
    WeightMinimum=0
    ExecutePin=@RigVMFunction_ControlFlowBranch.True
   )
   n7 = &RigVMFunction_ControlFlowBranch(ExecuteContext=@RigVMFunction_Sequence.D, Condition=RigVMVariableNode() {variable=isWearingPack}.Value)
   n8 = &RigUnit_SphericalPoseReader(
    ActiveRegionScaleFactors=(PositiveWidth=1,NegativeWidth=1,PositiveHeight=1,NegativeHeight=1)
    ActiveRegionSize=0.1
    Debug=(bDrawDebug=false,bDraw2D=False,bDrawLocalAxes=False,DebugScale=25,DebugSegments=20,DebugThickness=0.25)
    DriverAxis=(X=-1,Y=0,Z=0)
    DriverItem=(Type=Bone,Name="clavicle_r")
    FalloffRegionScaleFactors=(PositiveWidth=1,NegativeWidth=1,PositiveHeight=1,NegativeHeight=1)
    FalloffSize=0.2
    FlipHeightScaling=False
    FlipWidthScaling=False
    OptionalParentItem=(Type=Bone,Name="pelvis")
    RotationOffset=(X=134.5901,Y=-5.2068,Z=180.5049)
    ExecutePin=@n7.True
   )
   > &RigUnit_ModifyTransforms(
    ItemToModify=((Item=(Type=Bone,Name="dyn_r_hair_01"),Transform=(Rotation=(X=0,Y=0,Z=0,W=1),Translation=(X=0,Y=-4,Z=0),Scale3D=(X=1,Y=1,Z=1))))
    WeightMinimum=0
    Weight=@n8.OutputParam
   )
   RigVMFunctionReturnNode(ExecuteContext=@RigVMFunction_Sequence.E)
  ]}
 RigVMFunctionLibrary {nodes [
   Controllers = RigVMCollapseNode(ExecuteContext=(), debug=False)
   Hair Adjustments = RigVMCollapseNode(ExecuteContext=())
   SplineIK = RigVMCollapseNode(ExecuteContext=())
  ]}
 RigVMFunction_Sequence_1_ContainedGraph {nodes [
   n1 = RigVMFunctionEntryNode()
   RigVMFunction_Sequence = &RigVMFunction_Sequence(ExecuteContext=@n1.ExecuteContext)
   n2 = &RigVMFunction_Sequence(ExecuteContext=@RigVMFunction_Sequence.B)
   RigVMFunctionReturnNode(A=@RigVMFunction_Sequence.A, B=@n2.A, C=@n2.B)
  ]}
 RigVMFunction_Sequence_1_ContainedGraph {nodes [
   n1 = RigVMFunctionEntryNode()
   RigVMFunction_Sequence = &RigVMFunction_Sequence(ExecuteContext=@n1.ExecuteContext)
   n2 = &RigVMFunction_Sequence(ExecuteContext=@RigVMFunction_Sequence.B)
   n3 = &RigVMFunction_Sequence(ExecuteContext=@n2.B)
   n4 = &RigVMFunction_Sequence(ExecuteContext=@n3.B)
   RigVMFunctionReturnNode(A=@RigVMFunction_Sequence.A, B=@n2.A, C=@n3.A, D=@n4.A, E=@n4.B)
  ]}
 RigVMFunction_Sequence_2_ContainedGraph {nodes [
   n1 = RigVMFunctionEntryNode()
   RigVMFunction_Sequence = &RigVMFunction_Sequence(ExecuteContext=@n1.ExecuteContext)
   n2 = &RigVMFunction_Sequence(ExecuteContext=@RigVMFunction_Sequence.B)
   RigVMFunctionReturnNode(A=@RigVMFunction_Sequence.A, B=@n2.A, C=@n2.B)
  ]}
 RigVMModel {nodes [
   RigUnit_BeginExecution = &RigUnit_BeginExecution()
   SplineIK = RigVMFunctionReferenceNode(ExecuteContext=@RigUnit_BeginExecution.ExecutePin)
   > Hair Adjustments = RigVMFunctionReferenceNode()
  ]}
 RigVMModel Construction Event Graph {nodes [
   n1 = &RigUnit_PrepareForExecution()
   RigVMFunction_Sequence = RigVMAggregateNode(ExecuteContext=@n1.ExecutePin)
   n2 = &RigUnit_ItemArray(Items="((Type=Control,Name=\"dyn_l_hair_01_ctrl\"),(Type=Control,Name=\"dyn_l_hair_02_ctrl\"),(Type=Control,Name=\"dyn_l_hair_03_ctrl\"),(Type=Control,Name=\"dyn_c_hair_04_ctrl\"),(Type=Control,Name=\"dyn_l_hair_05_ctrl\"),(Type=Control,Name=\"dyn_l_hair_06_ctrl\"),(Type=Con ...(truncated)")
   RigVMVariableNode(Value=@n2.Items, ExecuteContext=@RigVMFunction_Sequence.A) {variable=l_dyn_hair_controls}
   > RigVMVariableNode(Value=/Script/ControlRig.RigUnit_HierarchyGetPoseItemArray(ElementType=All, Initial=False, ItemsToGet=@n2.Items) {is=struct RigVMUnitNode}.Pose) {variable=l_dyn_hair_pose}
   n3 = &RigUnit_ItemArray(Items="((Type=Control,Name=\"dyn_r_hair_01_ctrl\"),(Type=Control,Name=\"dyn_r_hair_02_ctrl\"),(Type=Control,Name=\"dyn_r_hair_03_ctrl\"),(Type=Control,Name=\"dyn_r_hair_04_ctrl\"),(Type=Control,Name=\"dyn_r_hair_05_ctrl\"),(Type=Control,Name=\"dyn_r_hair_06_ctrl\"),(Type=Con ...(truncated)")
   RigVMVariableNode(Value=@n3.Items, ExecuteContext=@RigVMFunction_Sequence.B) {variable=r_dyn_hair_controls}
   > RigVMVariableNode(Value=&RigUnit_HierarchyGetPoseItemArray(ElementType=All, Initial=False, ItemsToGet=@n3.Items).Pose) {variable=r_dyn_hair_pose}
   n4 = &RigUnit_ItemArray(Items="((Type=Control,Name=\"dyn_c_hair_01_ctrl\"),(Type=Control,Name=\"dyn_c_hair_02_ctrl\"),(Type=Control,Name=\"dyn_c_hair_03_ctrl\"),(Type=Control,Name=\"dyn_c_hair_04_ctrl\"),(Type=Control,Name=\"dyn_c_hair_05_ctrl\"),(Type=Control,Name=\"dyn_c_hair_06_ctrl\"),(Type=Con ...(truncated)")
   RigVMVariableNode(Value=@n4.Items, ExecuteContext=@RigVMFunction_Sequence.C) {variable=c_dyn_hair_controls}
   > RigVMVariableNode(Value=&RigUnit_HierarchyGetPoseItemArray(ElementType=All, Initial=False, ItemsToGet=@n4.Items).Pose) {variable=c_dyn_hair_pose}
  ]}
]
hierarchy [
 root {loc=<0 0 0>, rot=<0 0 0 1>}
 pelvis {parent=root, loc=<0 -0.5301 98.0571>, rot=<0.0127 0.707 -0.0127 -0.707>}
 spine_01 {parent=pelvis, loc=<6.1208 -0.9328 0>, rot=<0 0 0.034 -0.9994>}
 spine_02 {parent=spine_01, loc=<6.5996 -0.2095 0>, rot=<0 0 0.0238 -0.9997>}
 spine_03 {parent=spine_02, loc=<6.4914 -0.1237 0>, rot=<0 0 0.0149 -0.9999>}
 spine_04 {parent=spine_03, loc=<6.9598 0.0794 0>, rot=<0 0 -0.0132 -0.9999>}
 spine_05 {parent=spine_04, loc=<6.9987 0 0>, rot=<0 0 -0.0015 -1>}
 clavicle_l {parent=spine_05, loc=<5.4831 -0.1053 -3.2946>, rot=<0.0465 -0.7289 -0.009 -0.6829>}
 upperarm_l {parent=clavicle_l, loc=<9.8658 -1.2136 -0.4105>, rot=<0.0164 -0.3245 -0.0269 -0.9454>}
 lowerarm_l {parent=upperarm_l, loc=<26.465 -0.8345 0.259>, rot=<0.0036 0.0161 -0.2298 -0.9731>}
 neck_01 {parent=spine_05, loc=<11.3138 0.1529 0>, rot=<0 0 -0.2003 -0.9797>}
 neck_02 {parent=neck_01, loc=<4.5627 0 0>, rot=<0 0 0.0504 -0.9987>}
 head {parent=neck_02, loc=<5.3914 0 0>, rot=<0 0 0.115 -0.9934>}
 dyn_c_hair_mstr {parent=head, loc=<2.4304 -5.7143 0>, rot=<0.241 -0.9705 -0.0062 -0.0015>}
 dyn_c_hair_01 {parent=dyn_c_hair_mstr, loc=<0 0 0>, rot=<0 0 0 1>}
 dyn_c_hair_02 {parent=dyn_c_hair_01, loc=<6.6 0 0>, rot=<0 0 0.0551 -0.9985>}
 dyn_c_hair_03 {parent=dyn_c_hair_02, loc=<6.6 0 0>, rot=<0 0 -0.0929 -0.9957>}
 dyn_c_hair_04 {parent=dyn_c_hair_03, loc=<6.6312 0 0>, rot=<0 0 -0.1118 -0.9937>}
 dyn_c_hair_05 {parent=dyn_c_hair_04, loc=<6.6227 0 0>, rot=<0 0 -0.0776 -0.997>}
 dyn_c_hair_06 {parent=dyn_c_hair_05, loc=<7.1605 0 0>, rot=<0 0 -0.0429 -0.9991>}
 dyn_c_hair_07 {parent=dyn_c_hair_06, loc=<6.6381 0 0>, rot=<0 0 -0.0827 -0.9966>}
 dyn_l_hair_mstr {parent=head, loc=<2.5 -5.1499 -6>, rot=<0.241 -0.9705 -0.0062 -0.0015>}
 dyn_l_hair_01 {parent=dyn_l_hair_mstr, loc=<0 0 0>, rot=<0 0 0 1>}
 dyn_l_hair_02 {parent=dyn_l_hair_01, loc=<6.6 0 0>, rot=<0 0 0.0551 -0.9985>}
 dyn_l_hair_03 {parent=dyn_l_hair_02, loc=<6.6 0 0>, rot=<0 0 -0.0929 -0.9957>}
 dyn_l_hair_04 {parent=dyn_l_hair_03, loc=<6.6312 0 0>, rot=<0 0 -0.1118 -0.9937>}
 dyn_l_hair_05 {parent=dyn_l_hair_04, loc=<6.6227 0 0>, rot=<0 0 -0.0776 -0.997>}
 dyn_l_hair_06 {parent=dyn_l_hair_05, loc=<7.1605 0 0>, rot=<0 0 -0.0429 -0.9991>}
 dyn_l_hair_07 {parent=dyn_l_hair_06, loc=<6.6381 0 0>, rot=<0 0 -0.0827 -0.9966>}
 dyn_r_hair_mstr {parent=head, loc=<2.5 -5.1499 6>, rot=<0.241 -0.9705 -0.0062 -0.0015>}
 dyn_r_hair_01 {parent=dyn_r_hair_mstr, loc=<0 0 0>, rot=<0 0 0 1>}
 dyn_r_hair_02 {parent=dyn_r_hair_01, loc=<6.6 0 0>, rot=<0 0 0.0551 -0.9985>}
 dyn_r_hair_03 {parent=dyn_r_hair_02, loc=<6.6 0 0>, rot=<0 0 -0.0929 -0.9957>}
 dyn_r_hair_04 {parent=dyn_r_hair_03, loc=<6.6312 0 0>, rot=<0 0 -0.1118 -0.9937>}
 dyn_r_hair_05 {parent=dyn_r_hair_04, loc=<6.6227 0 0>, rot=<0 0 -0.0776 -0.997>}
 dyn_r_hair_06 {parent=dyn_r_hair_05, loc=<7.1605 0 0>, rot=<0 0 -0.0429 -0.9991>}
 dyn_r_hair_07 {parent=dyn_r_hair_06, loc=<6.6381 0 0>, rot=<0 0 -0.0827 -0.9966>}
 clavicle_r {parent=spine_05, loc=<5.4831 -0.1053 3.2946>, rot=<0.7289 0.0465 -0.6829 0.009>}
 upperarm_r {parent=clavicle_r, loc=<-9.8658 1.2136 0.4105>, rot=<0.0164 -0.3245 -0.0269 -0.9454>}
 lowerarm_r {parent=upperarm_r, loc=<-26.465 0.8345 -0.259>, rot=<0.0036 0.0161 -0.2298 -0.9731>}
 dyn_c_hair_01_null {type=Null, loc=<0 -7.4215 154.4174>, rot=<0.1719 -0.6859 0.1632 -0.688>}
 dyn_c_hair_01_ctrl {controlType=EULER_TRANSFORM, parent=dyn_c_hair_01_null, type=Control, loc=<0 0 0>, rot=<0 0 0 1>}
 dyn_c_hair_02_null {type=Null, loc=<0.0385 -10.4604 148.5587>, rot=<-0.2095 0.6754 -0.2009 0.678>}
 dyn_c_hair_02_ctrl {controlType=EULER_TRANSFORM, parent=dyn_c_hair_02_null, type=Control, loc=<0 0 0>, rot=<0 0 0 1>}
 dyn_c_hair_03_null {type=Null, loc=<0.0849 -14.1259 143.0704>, rot=<0.1459 -0.6919 0.1371 -0.6937>}
 dyn_c_hair_03_ctrl {controlType=EULER_TRANSFORM, parent=dyn_c_hair_03_null, type=Control, loc=<0 0 0>, rot=<0 0 0 1>}
 dyn_c_hair_04_null {type=Null, loc=<0.1178 -16.7255 136.9701>, rot=<-0.0676 0.7039 -0.0587 0.7047>}
 dyn_c_hair_04_ctrl {controlType=EULER_TRANSFORM, parent=dyn_c_hair_04_null, type=Control, loc=<0 0 0>, rot=<0 0 0 1>}
 dyn_c_hair_05_null {type=Null, loc=<0.1327 -17.9038 130.4531>, rot=<0.0128 -0.707 0.0038 -0.7071>}
 dyn_c_hair_05_ctrl {controlType=EULER_TRANSFORM, parent=dyn_c_hair_05_null, type=Control, loc=<0 0 0>, rot=<0 0 0 1>}
 dyn_c_hair_06_null {type=Null, loc=<0.1349 -18.0723 123.2945>, rot=<0.0176 0.7069 0.0265 0.7066>}
 dyn_c_hair_06_ctrl {controlType=EULER_TRANSFORM, parent=dyn_c_hair_06_null, type=Control, loc=<0 0 0>, rot=<0 0 0 1>}
 dyn_c_hair_07_null {type=Null, loc=<0.1296 -17.6586 116.6693>, rot=<-0.076 -0.703 -0.0849 -0.702>}
 dyn_c_hair_07_ctrl {controlType=EULER_TRANSFORM, parent=dyn_c_hair_07_null, type=Control, loc=<0 0 0>, rot=<0 0 0 1>}
 dyn_r_hair_01_null {type=Null, loc=<-6 -6.8578 154.4916>, rot=<0.1719 -0.6859 0.1632 -0.688>}
 dyn_r_hair_01_ctrl {controlType=EULER_TRANSFORM, parent=dyn_r_hair_01_null, type=Control, loc=<0 0 0>, rot=<0 0 0 1>}
 dyn_r_hair_02_null {type=Null, loc=<-5.9615 -9.8967 148.6329>, rot=<-0.2095 0.6754 -0.2009 0.678>}
 dyn_r_hair_02_ctrl {controlType=EULER_TRANSFORM, parent=dyn_r_hair_02_null, type=Control, loc=<0 0 0>, rot=<0 0 0 1>}
 dyn_r_hair_03_null {type=Null, loc=<-5.9151 -13.5621 143.1446>, rot=<0.1459 -0.6919 0.1371 -0.6937>}
 dyn_r_hair_03_ctrl {controlType=EULER_TRANSFORM, parent=dyn_r_hair_03_null, type=Control, loc=<0 0 0>, rot=<0 0 0 1>}
 dyn_r_hair_04_null {type=Null, loc=<-5.8822 -16.1618 137.0443>, rot=<-0.0676 0.7039 -0.0587 0.7047>}
 dyn_r_hair_04_ctrl {controlType=EULER_TRANSFORM, parent=dyn_r_hair_04_null, type=Control, loc=<0 0 0>, rot=<0 0 0 1>}
 dyn_r_hair_05_null {type=Null, loc=<-5.8673 -17.34 130.5273>, rot=<0.0128 -0.707 0.0038 -0.7071>}
 dyn_r_hair_05_ctrl {controlType=EULER_TRANSFORM, parent=dyn_r_hair_05_null, type=Control, loc=<0 0 0>, rot=<0 0 0 1>}
 dyn_r_hair_06_null {type=Null, loc=<-5.8651 -17.5086 123.3687>, rot=<0.0176 0.7069 0.0265 0.7066>}
 dyn_r_hair_06_ctrl {controlType=EULER_TRANSFORM, parent=dyn_r_hair_06_null, type=Control, loc=<0 0 0>, rot=<0 0 0 1>}
 dyn_r_hair_07_null {type=Null, loc=<-5.8704 -17.0948 116.7435>, rot=<-0.076 -0.703 -0.0849 -0.702>}
 dyn_r_hair_07_ctrl {controlType=EULER_TRANSFORM, parent=dyn_r_hair_07_null, type=Control, loc=<0 0 0>, rot=<0 0 0 1>}
 dyn_l_hair_01_null {type=Null, loc=<6 -6.8578 154.4916>, rot=<0.1719 -0.6859 0.1632 -0.688>}
 dyn_l_hair_01_ctrl {controlType=EULER_TRANSFORM, parent=dyn_l_hair_01_null, type=Control, loc=<0 0 0>, rot=<0 0 0 1>}
 dyn_l_hair_02_null {type=Null, loc=<6.0385 -9.8967 148.6329>, rot=<-0.2095 0.6754 -0.2009 0.678>}
 dyn_l_hair_02_ctrl {controlType=EULER_TRANSFORM, parent=dyn_l_hair_02_null, type=Control, loc=<0.5316 -9.8984 148.6329>, rot=<-0.2464 0.6628 -0.1632 0.688>}
 dyn_l_hair_03_null {type=Null, loc=<6.0849 -13.5621 143.1446>, rot=<0.1459 -0.6919 0.1371 -0.6937>}
 dyn_l_hair_03_ctrl {controlType=EULER_TRANSFORM, parent=dyn_l_hair_03_null, type=Control, loc=<-3.0141 -13.233 143.1446>, rot=<-0.081 0.7025 -0.2009 0.678>}
 dyn_l_hair_04_null {type=Null, loc=<6.1178 -16.1618 137.0443>, rot=<-0.0676 0.7039 -0.0587 0.7047>}
 dyn_l_hair_04_ctrl {controlType=EULER_TRANSFORM, parent=dyn_l_hair_04_null, type=Control, loc=<-4.0905 -15.644 137.0443>, rot=<0.0115 0.707 -0.1371 0.6937>}
 dyn_l_hair_05_null {type=Null, loc=<6.1327 -17.34 130.5273>, rot=<0.0128 -0.707 0.0038 -0.7071>}
 dyn_l_hair_05_ctrl {controlType=EULER_TRANSFORM, parent=dyn_l_hair_05_null, type=Control, loc=<-3.1668 -17.0555 130.5273>, rot=<0.0421 0.7059 -0.0587 0.7047>}
 dyn_l_hair_06_null {type=Null, loc=<6.1349 -17.5086 123.3687>, rot=<0.0176 0.7069 0.0265 0.7066>}
 dyn_l_hair_06_ctrl {controlType=EULER_TRANSFORM, parent=dyn_l_hair_06_null, type=Control, loc=<-2.5245 -17.356 123.3687>, rot=<0.0479 0.7055 -0.0038 0.7071>}
 dyn_l_hair_07_null {type=Null, loc=<6.1296 -17.0948 116.7435>, rot=<-0.076 -0.703 -0.0849 -0.702>}
 dyn_l_hair_07_ctrl {controlType=EULER_TRANSFORM, parent=dyn_l_hair_07_null, type=Control, loc=<-3.3198 -16.7771 116.7435>, rot=<0.1339 0.6943 0.0265 0.7066>}
]
variables [
 c_dyn_hair_controls {
  type="TArray<FRigElementKey>"
  default="((Type=Control,Name=\"dyn_l_hair_01_ctrl\"),(Type=Control,Name=\"dyn_l_hair_02_ctrl\"),(Type=Control,Name=\"dyn_l_hair_03_ctrl\"),(Type=Control,Name=\"dyn_c_hair_04_ctrl\"),(Type=Control,Name=\"dyn_l_hair_05_ctrl\"),(Type=Control,Name=\"dyn_l_hair_06_ctrl\"),(Type=Con ...(truncated)"
 }
 c_dyn_hair_pose {type=FRigPose, default=(Elements=,HierarchyTopologyVersion=-1,PoseHash=-1)}
 isFrontEnd {type=bool, default=False}
 isSkydiving {type=bool, default=False}
 isWearingPack {type=bool, default=False}
 l_dyn_hair_controls {
  type="TArray<FRigElementKey>"
  default="((Type=Control,Name=\"dyn_l_hair_01_ctrl\"),(Type=Control,Name=\"dyn_l_hair_02_ctrl\"),(Type=Control,Name=\"dyn_l_hair_03_ctrl\"),(Type=Control,Name=\"dyn_c_hair_04_ctrl\"),(Type=Control,Name=\"dyn_l_hair_05_ctrl\"),(Type=Control,Name=\"dyn_l_hair_06_ctrl\"),(Type=Con ...(truncated)"
 }
 l_dyn_hair_pose {type=FRigPose, default=(Elements=,HierarchyTopologyVersion=-1,PoseHash=-1)}
 r_dyn_hair_controls {
  type="TArray<FRigElementKey>"
  default="((Type=Control,Name=\"dyn_l_hair_01_ctrl\"),(Type=Control,Name=\"dyn_l_hair_02_ctrl\"),(Type=Control,Name=\"dyn_l_hair_03_ctrl\"),(Type=Control,Name=\"dyn_c_hair_04_ctrl\"),(Type=Control,Name=\"dyn_l_hair_05_ctrl\"),(Type=Control,Name=\"dyn_l_hair_06_ctrl\"),(Type=Con ...(truncated)"
 }
 r_dyn_hair_pose {type=FRigPose, default=(Elements=,HierarchyTopologyVersion=-1,PoseHash=-1)}
]
