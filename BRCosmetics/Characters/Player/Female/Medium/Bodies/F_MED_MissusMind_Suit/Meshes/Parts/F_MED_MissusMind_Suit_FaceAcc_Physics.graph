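# Physics asset for the face-accessory mesh of the female medium MissusMind suit, setting up its dynamic hair bones.
# Center, left and right hair chains each link four capsule bodies to the head, with limited swing and zero-degree twist limits.
# Capsules on pelvis, spine, neck, head, clavicles and upper arms form body volumes around the hair.
# Those body joints have free angular motion, with upper spine, neck and head linked directly to spine_02.
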PhysicsAsset /BRCosmetics/Characters/Player/Female/Medium/Bodies/F_MED_MissusMind_Suit/Meshes/Parts/F_MED_MissusMind_Suit_FaceAcc_Physics
bodies {
 clavicle_l [{
   shapeType=Capsule
   length=8.4565
   radius=7.4147
   shapeName=clavicle_l_capsule
   center=<4.4026 0 -1.1441>
   rotation=<rot 83.5971 -0.0005 -0.0005>
  }]
 clavicle_r [{
   shapeType=Capsule
   length=9.3473
   radius=7.1946
   shapeName=clavicle_r_capsule
   center=<-4.7964 0 1.1574>
   rotation=<rot -84.6834 179.999 -179.999>
  }]
 dyn_c_hair_04 [{shapeType=Capsule, length=0.5, radius=0.5, shapeName=dyn_c_hair_04_capsule, rotation=<rot -90 180 180>}]
 dyn_c_hair_05 [{shapeType=Capsule, length=4.7483, radius=3.1814, shapeName=dyn_c_hair_05_capsule, center=<3.8537 -1.2 0>, rotation=<rot -90 180 180>}]
 dyn_c_hair_06 [{shapeType=Capsule, length=5.9637, radius=3.7137, shapeName=dyn_c_hair_05_capsule, center=<3.8537 -1.8 0>, rotation=<rot -90 180 180>}]
 dyn_c_hair_07 [{shapeType=Capsule, length=4.247, radius=3.6828, shapeName=dyn_c_hair_05_capsule, center=<4.5567 -2.5461 0>, rotation=<rot -90 0 0>}]
 dyn_l_hair_04 [{shapeType=Capsule, length=0.5, radius=0.5, shapeName=dyn_l_hair_04_capsule, rotation=<rot -90 0 0>}]
 dyn_l_hair_05 [{shapeType=Capsule, length=4.3203, radius=3, shapeName=dyn_l_hair_05_capsule, center=<3.8537 -1.2 0>, rotation=<rot -90 0 0>}]
 dyn_l_hair_06 [{shapeType=Capsule, length=4.7059, radius=3.5083, shapeName=dyn_l_hair_04_capsule, center=<2.4829 -1.8 0>, rotation=<rot -90 180 179.0002>}]
 dyn_l_hair_07 [{shapeType=Capsule, length=5.3414, radius=3.9789, shapeName=dyn_l_hair_05_capsule, center=<3.3584 -2.6187 0>, rotation=<rot -90 180 177.9844>}]
 dyn_r_hair_04 [{shapeType=Capsule, length=0.5, radius=0.5, shapeName=dyn_r_hair_04_capsule, rotation=<rot -90 0 0>}]
 dyn_r_hair_05 [{shapeType=Capsule, length=4.4975, radius=3, shapeName=dyn_r_hair_05_capsule, center=<3.8537 -1.2 0>, rotation=<rot -90 0 0>}]
 dyn_r_hair_06 [{shapeType=Capsule, length=4.7059, radius=3.5083, shapeName=dyn_r_hair_04_capsule, center=<2.5338 -1.8 0>, rotation=<rot -90 180 180>}]
 dyn_r_hair_07 [{shapeType=Capsule, length=5.7314, radius=3.7661, shapeName=dyn_r_hair_05_capsule, center=<3.7517 -2.5461 0>, rotation=<rot -90 0 0>}]
 head [{shapeType=Capsule, length=5.4047, radius=7.0498, shapeName=head_capsule, center=<4.1921 0.0665 0>, rotation=<rot 90 0 0>}]
 neck_02 [{shapeType=Capsule, length=10.1043, radius=6.9263, shapeName=neck_02_capsule, rotation=<rot 90 0 0>}]
 pelvis [{shapeType=Capsule, length=15.0317, radius=11.9493, shapeName=pelvis_capsule, center=<-3.1244 -1.4683 0>, rotation=<rot 0 -90 0>}]
 spine_01 [{shapeType=Capsule, length=16.3134, radius=9.1081, shapeName=spine_01_capsule, rotation=<rot 0 -90 0>}]
 spine_02 [{shapeType=Capsule, length=17.4425, radius=7.9791, shapeName=spine_02_capsule, rotation=<rot 0 -90 0>}]
 spine_03 [{shapeType=Capsule, length=8.4511, radius=9.1042, shapeName=spine_03_capsule, center=<2.1937 1.0925 0>, rotation=<rot 0 -90 0>}]
 spine_04 [{shapeType=Capsule, length=10.1504, radius=10.8738, shapeName=spine_04_capsule, center=<2.1937 1.5397 0>, rotation=<rot 0 -90 0>}]
 spine_05 [{shapeType=Capsule, length=11.2825, radius=12.1028, shapeName=spine_05_capsule, center=<1.4926 -0.1837 0>, rotation=<rot 0 -90 0>}]
 upperarm_l [{shapeType=Capsule, length=29.1771, radius=5.2279, shapeName=upperarm_l_capsule, center=<11.2265 -0.6167 -0.3866>, rotation=<rot -90 180 180>}]
 upperarm_r [{shapeType=Capsule, length=29.1771, radius=5.2279, shapeName=upperarm_r_capsule, center=<-11.2265 0.6603 0.3866>, rotation=<rot 90 0 0>}]
}
constraints [
 {
  bone1Name=clavicle_l
  bone2Name=spine_05
  swing1LimitDegrees=45
  swing1Motion=Free
  swing2LimitDegrees=45
  swing2Motion=Free
  twistLimitDegrees=45
  twistMotion=Free
 }
 {
  bone1Name=clavicle_r
  bone2Name=spine_05
  swing1LimitDegrees=45
  swing1Motion=Free
  swing2LimitDegrees=45
  swing2Motion=Free
  twistLimitDegrees=45
  twistMotion=Free
 }
 {bone1Name=dyn_c_hair_05, bone2Name=dyn_c_hair_04, swing1LimitDegrees=20, swing2LimitDegrees=20, twistLimitDegrees=0}
 {bone1Name=dyn_c_hair_04, bone2Name=head, swing1LimitDegrees=20, swing2LimitDegrees=12, twistLimitDegrees=0}
 {bone1Name=dyn_l_hair_05, bone2Name=dyn_l_hair_04, swing1LimitDegrees=20, swing2LimitDegrees=20, twistLimitDegrees=0}
 {bone1Name=dyn_l_hair_04, bone2Name=head, swing1LimitDegrees=20, swing2LimitDegrees=12, twistLimitDegrees=0}
 {bone1Name=dyn_r_hair_05, bone2Name=dyn_r_hair_04, swing1LimitDegrees=20, swing2LimitDegrees=20, twistLimitDegrees=0}
 {bone1Name=dyn_r_hair_04, bone2Name=head, swing1LimitDegrees=20, swing2LimitDegrees=12, twistLimitDegrees=0}
 {
  bone1Name=spine_03
  bone2Name=spine_02
  swing1LimitDegrees=45
  swing1Motion=Free
  swing2LimitDegrees=45
  swing2Motion=Free
  twistLimitDegrees=45
  twistMotion=Free
 }
 {bone1Name=dyn_c_hair_07, bone2Name=dyn_c_hair_06, swing1LimitDegrees=30, swing2LimitDegrees=30, twistLimitDegrees=0}
 {bone1Name=dyn_l_hair_07, bone2Name=dyn_l_hair_06, swing1LimitDegrees=30, swing2LimitDegrees=30, twistLimitDegrees=0}
 {bone1Name=dyn_r_hair_07, bone2Name=dyn_r_hair_06, swing1LimitDegrees=30, swing2LimitDegrees=30, twistLimitDegrees=0}
 {bone1Name=dyn_c_hair_06, bone2Name=dyn_c_hair_05, swing1LimitDegrees=20, swing2LimitDegrees=20, twistLimitDegrees=0}
 {bone1Name=dyn_r_hair_06, bone2Name=dyn_r_hair_05, swing1LimitDegrees=20, swing2LimitDegrees=20, twistLimitDegrees=0}
 {bone1Name=dyn_l_hair_06, bone2Name=dyn_l_hair_05, swing1LimitDegrees=20, swing2LimitDegrees=20, twistLimitDegrees=0}
 {
  bone1Name=spine_04
  bone2Name=spine_02
  swing1LimitDegrees=45
  swing1Motion=Free
  swing2LimitDegrees=45
  swing2Motion=Free
  twistLimitDegrees=45
  twistMotion=Free
 }
 {
  bone1Name=spine_05
  bone2Name=spine_02
  swing1LimitDegrees=45
  swing1Motion=Free
  swing2LimitDegrees=45
  swing2Motion=Free
  twistLimitDegrees=45
  twistMotion=Free
 }
 {
  bone1Name=neck_02
  bone2Name=spine_02
  swing1LimitDegrees=45
  swing1Motion=Free
  swing2LimitDegrees=45
  swing2Motion=Free
  twistLimitDegrees=45
  twistMotion=Free
 }
 {
  bone1Name=head
  bone2Name=spine_02
  swing1LimitDegrees=45
  swing1Motion=Free
  swing2LimitDegrees=45
  swing2Motion=Free
  twistLimitDegrees=45
  twistMotion=Free
 }
 {
  bone1Name=spine_02
  bone2Name=spine_01
  swing1LimitDegrees=45
  swing1Motion=Free
  swing2LimitDegrees=45
  swing2Motion=Free
  twistLimitDegrees=45
  twistMotion=Free
 }
 {
  bone1Name=spine_01
  bone2Name=pelvis
  swing1LimitDegrees=45
  swing1Motion=Free
  swing2LimitDegrees=45
  swing2Motion=Free
  twistLimitDegrees=45
  twistMotion=Free
 }
 {
  bone1Name=upperarm_l
  bone2Name=clavicle_l
  swing1LimitDegrees=45
  swing1Motion=Free
  swing2LimitDegrees=45
  swing2Motion=Free
  twistLimitDegrees=45
  twistMotion=Free
 }
 {
  bone1Name=upperarm_r
  bone2Name=clavicle_r
  swing1LimitDegrees=45
  swing1Motion=Free
  swing2LimitDegrees=45
  swing2Motion=Free
  twistLimitDegrees=45
  twistMotion=Free
 }
]
