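# Anim Blueprint for MissusMind Suit's FaceAcc part, animating three hair chains in game.
# Copies the attached parent's pose into a Control Rig, then caches two variants: backpack (bone offsets, tip AnimDynamics) and dynamic (full-chain AnimDynamics, head-space RigidBody).
# Nested bool blends select between these and the raw rig pose for output.
# EventGraph derives gravity, skydiving-dependent physics strength and backbling flags from character state.
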AnimBlueprint /BRCosmetics/Characters/Player/Female/Medium/Bodies/F_MED_MissusMind_Suit/Meshes/Parts/F_MED_MissusMind_Suit_FaceAcc_AnimBP
target_skeleton=&/F_MED_MissusMind_Suit_FaceAcc_Skeleton
graphs [
 AnimGraph {nodes [
   RigidBody!(ComponentPose, Alpha, OverrideWorldGravity:Vector="0,0,0", ExternalForce) -> (Pose) {
    bClampLinearTranslationLimitToRefPose=true
    bOverrideWorldGravity=true
    overridePhysicsAsset=&/F_MED_MissusMind_Suit_FaceAcc_Physics
    simulationSpace=BaseBoneSpace
    alphaBoolBlend {blendOption=Linear}
    baseBoneRef {boneName=head}
    componentAppliedLinearAccClamp=<1000 1000 1000>
    componentLinearAccScale=<0.3 0.4 0.4>
    componentLinearVelScale=<0.3 0.4 0.4>
    simSpaceSettings {maxAngularAcceleration=5, maxAngularVelocity=10, maxLinearAcceleration=50, maxLinearVelocity=150, worldAlpha=0.2}
   }
   Output Pose(Result=Use cached pose 'Blend_Hair_Pose_Cache'()) {layerGroup=None}
   n1 = Control Rig!(
    Source=Copy Pose From Mesh() {bUseAttachedParent=true}
    Alpha
    isSkydiving:bool=Get bIsSkyDiving()
    isFrontEnd:bool=Get bIsInFrontEnd()
    isWearingPack:bool=Get isWearingPack() -> (isWearingPack:bool)
   ) -> (Pose) {
    alphaBoolBlend {blendOption=Linear}
    controlRigAssetReference {blueprintRigClass=&/F_MED_MissusMind_Suit_FaceAcc_CtrlRig1.F_MED_MissusMind_Suit_FaceAcc_CtrlRig1_C}
   }
   Save cached pose 'CR_HairPose_Cache'(Pose=@n1) {cacheName=CR_HairPose_Cache}
   n2 = "Transform (Modify) Bone - Bone: dyn_c_hair_02"(
    ComponentPose=Local To Component(LocalPose=Use cached pose 'CR_HairPose_Cache'())
    Alpha=1
    Translation="0,-3,0"
    Rotation="0,6,0"
    Scale="1,1,1"
   ) {
    rotationMode=BMM_Additive
    rotationSpace=BCS_BoneSpace
    translationMode=BMM_Additive
    translationSpace=BCS_BoneSpace
    alphaBoolBlend {blendOption=Linear}
    boneToModify {boneName=dyn_c_hair_02}
    rotation {yaw=6}
    translation {y=-3}
   }
   n3 = "Transform (Modify) Bone - Bone: dyn_l_hair_02"(ComponentPose=@n2, Alpha=1, Translation="0,-1,0", Rotation="0,6,0", Scale="1,1,1") {
    rotationMode=BMM_Additive
    rotationSpace=BCS_BoneSpace
    translationMode=BMM_Additive
    translationSpace=BCS_BoneSpace
    alphaBoolBlend {blendOption=Linear}
    boneToModify {boneName=dyn_l_hair_02}
    rotation {yaw=6}
    translation {y=-1}
   }
   n4 = "Transform (Modify) Bone - Bone: dyn_r_hair_02"(ComponentPose=@n3, Alpha=1, Translation="0,-1,0", Rotation="0,6,0", Scale="1,1,1") {
    rotationMode=BMM_Additive
    rotationSpace=BCS_BoneSpace
    translationMode=BMM_Additive
    translationSpace=BCS_BoneSpace
    alphaBoolBlend {blendOption=Linear}
    boneToModify {boneName=dyn_r_hair_02}
    rotation {yaw=6}
    translation {y=-1}
   }
   n5 = "Transform (Modify) Bone - Bone: dyn_c_hair_03"(ComponentPose=@n4, Alpha=1, Translation="0,-1,0", Rotation="0,-2,0", Scale="1,1,1") {
    rotationMode=BMM_Additive
    rotationSpace=BCS_BoneSpace
    translationMode=BMM_Additive
    translationSpace=BCS_BoneSpace
    alphaBoolBlend {blendOption=Linear}
    boneToModify {boneName=dyn_c_hair_03}
    rotation {yaw=-2}
    translation {y=-1}
   }
   n6 = "Transform (Modify) Bone - Bone: dyn_l_hair_03"(ComponentPose=@n5, Alpha=1, Translation="0,-1,0", Rotation="0,-2,0", Scale="1,1,1") {
    rotationMode=BMM_Additive
    rotationSpace=BCS_BoneSpace
    translationMode=BMM_Additive
    translationSpace=BCS_BoneSpace
    alphaBoolBlend {blendOption=Linear}
    boneToModify {boneName=dyn_l_hair_03}
    rotation {yaw=-2}
    translation {y=-1}
   }
   n7 = "Transform (Modify) Bone - Bone: dyn_r_hair_03"(ComponentPose=@n6, Alpha=1, Translation="0,-1,0", Rotation="0,-2,0", Scale="1,1,1") {
    rotationMode=BMM_Additive
    rotationSpace=BCS_BoneSpace
    translationMode=BMM_Additive
    translationSpace=BCS_BoneSpace
    alphaBoolBlend {blendOption=Linear}
    boneToModify {boneName=dyn_r_hair_03}
    rotation {yaw=-2}
    translation {y=-1}
   }
   n8 = "Transform (Modify) Bone - Bone: dyn_c_hair_04"(ComponentPose=@n7, Alpha=1, Translation="0,2,0", Rotation="0,6,0", Scale="1,1,1") {
    rotationMode=BMM_Additive
    rotationSpace=BCS_BoneSpace
    translationMode=BMM_Additive
    translationSpace=BCS_BoneSpace
    alphaBoolBlend {blendOption=Linear}
    boneToModify {boneName=dyn_c_hair_04}
    rotation {yaw=6}
    translation {y=2}
   }
   n9 = "Transform (Modify) Bone - Bone: dyn_l_hair_04"(ComponentPose=@n8, Alpha=1, Translation="0,0,0", Rotation="0,6,0", Scale="1,1,1") {
    rotationMode=BMM_Additive
    rotationSpace=BCS_BoneSpace
    translationMode=BMM_Additive
    translationSpace=BCS_BoneSpace
    alphaBoolBlend {blendOption=Linear}
    boneToModify {boneName=dyn_l_hair_04}
    rotation {yaw=6}
   }
   n10 = "Transform (Modify) Bone - Bone: dyn_r_hair_04"(ComponentPose=@n9, Alpha=1, Translation="0,0,0", Rotation="0,6,0", Scale="1,1,1") {
    rotationMode=BMM_Additive
    rotationSpace=BCS_BoneSpace
    translationMode=BMM_Additive
    translationSpace=BCS_BoneSpace
    alphaBoolBlend {blendOption=Linear}
    boneToModify {boneName=dyn_r_hair_04}
    rotation {yaw=6}
   }
   n11 = Anim Dynamics - Bone: dyn_c_hair_06!(ComponentPose=@n10, Alpha=1, GravityOverride:Vector="0,0,-980", ExternalForce="0,0,0") -> (Pose) {
    angularDampingOverride=1
    angularSpringConstant=20000
    bAngularSpring=true
    bOverrideAngularDamping=true
    bUseGravityOverride=true
    bUseSphericalLimits=true
    gravityScale=0.35
    simulationSpace=RootRelative
    alphaBoolBlend {blendInTime=2, blendOption=Cubic}
    boundBone {boneName=dyn_c_hair_06}
    componentAppliedLinearAccClamp=<400000 400000 400000>
    componentLinearAccScale=<0.3 0.4 0.3>
    componentLinearVelScale=<0.3 0.4 3>
    physicsBodyDefinitions [{
      boundBone {boneName=dyn_c_hair_06}
      boxExtents=<25 25 25>
      constraintSetup {angularLimitsMax=<0 45 0>, angularLimitsMin=<0 0 0>, angularTarget=<1 0 0>}
      localJointOffset=<7 0 0>
     }]
    relativeSpaceBone {boneName=root}
    sphericalLimits [{limitRadius=15, drivingBone {boneName=pelvis}, sphereLocalOffset=<0 0 0>}]
   }
   n12 = Anim Dynamics - Bone: dyn_l_hair_06!(ComponentPose=@n11, Alpha=1, GravityOverride:Vector="0,0,-980", ExternalForce="0,0,0") -> (Pose) {
    angularDampingOverride=1
    angularSpringConstant=20000
    bAngularSpring=true
    bOverrideAngularDamping=true
    bUseGravityOverride=true
    bUseSphericalLimits=true
    gravityScale=0.35
    simulationSpace=RootRelative
    alphaBoolBlend {blendInTime=2, blendOption=Cubic}
    boundBone {boneName=dyn_l_hair_06}
    componentAppliedLinearAccClamp=<400000 400000 400000>
    componentLinearAccScale=<0.3 0.4 0.3>
    componentLinearVelScale=<0.3 0.4 3>
    physicsBodyDefinitions [{
      boundBone {boneName=dyn_l_hair_06}
      boxExtents=<25 25 25>
      constraintSetup {angularLimitsMax=<0 45 0>, angularLimitsMin=<0 0 0>, angularTarget=<1 0 0>}
      localJointOffset=<7 0 0>
     }]
    relativeSpaceBone {boneName=root}
    sphericalLimits [{limitRadius=15, drivingBone {boneName=pelvis}, sphereLocalOffset=<0 0 0>}]
   }
   n13 = Anim Dynamics - Bone: dyn_r_hair_06!(ComponentPose=@n12, Alpha=1, GravityOverride:Vector="0,0,-980", ExternalForce="0,0,0") -> (Pose) {
    angularDampingOverride=1
    angularSpringConstant=20000
    bAngularSpring=true
    bOverrideAngularDamping=true
    bUseGravityOverride=true
    bUseSphericalLimits=true
    gravityScale=0.35
    simulationSpace=RootRelative
    alphaBoolBlend {blendInTime=2, blendOption=Cubic}
    boundBone {boneName=dyn_r_hair_06}
    componentAppliedLinearAccClamp=<400000 400000 400000>
    componentLinearAccScale=<0.3 0.4 0.3>
    componentLinearVelScale=<0.3 0.4 3>
    physicsBodyDefinitions [{
      boundBone {boneName=dyn_r_hair_06}
      boxExtents=<25 25 25>
      constraintSetup {angularLimitsMax=<0 45 0>, angularLimitsMin=<0 0 0>, angularTarget=<1 0 0>}
      localJointOffset=<7 0 0>
     }]
    relativeSpaceBone {boneName=root}
    sphericalLimits [{limitRadius=15, drivingBone {boneName=pelvis}, sphereLocalOffset=<0 0 0>}]
   }
   n14 = Anim Dynamics - Bone: dyn_c_hair_07!(ComponentPose=@n13, Alpha=1, GravityOverride:Vector="0,0,-980", ExternalForce="0,0,0") -> (Pose) {
    angularDampingOverride=1
    angularSpringConstant=20000
    bAngularSpring=true
    bOverrideAngularDamping=true
    bUseGravityOverride=true
    bUseSphericalLimits=true
    gravityScale=0.35
    alphaBoolBlend {blendInTime=2, blendOption=Cubic}
    boundBone {boneName=dyn_c_hair_07}
    componentAppliedLinearAccClamp=<400000 400000 400000>
    componentLinearAccScale=<0.3 0.4 0.3>
    componentLinearVelScale=<0.3 0.4 3>
    physicsBodyDefinitions [{
      boundBone {boneName=dyn_c_hair_07}
      boxExtents=<25 25 25>
      constraintSetup {angularLimitsMax=<0 25 0>, angularLimitsMin=<0 -10 0>, angularTarget=<1 0 0>}
      localJointOffset=<7 0 0>
     }]
    relativeSpaceBone {boneName=root}
    sphericalLimits [{limitRadius=15, drivingBone {boneName=pelvis}, sphereLocalOffset=<0 0 0>}]
   }
   n15 = Anim Dynamics - Bone: dyn_l_hair_07!(ComponentPose=@n14, Alpha=1, GravityOverride:Vector="0,0,-980", ExternalForce="0,0,0") -> (Pose) {
    angularDampingOverride=1
    angularSpringConstant=20000
    bAngularSpring=true
    bOverrideAngularDamping=true
    bUseGravityOverride=true
    bUseSphericalLimits=true
    gravityScale=0.35
    alphaBoolBlend {blendInTime=2, blendOption=Cubic}
    boundBone {boneName=dyn_l_hair_07}
    componentAppliedLinearAccClamp=<400000 400000 400000>
    componentLinearAccScale=<0.3 0.4 0.3>
    componentLinearVelScale=<0.3 0.4 3>
    physicsBodyDefinitions [{
      boundBone {boneName=dyn_l_hair_07}
      boxExtents=<25 25 25>
      constraintSetup {angularLimitsMax=<0 25 0>, angularLimitsMin=<0 -10 0>, angularTarget=<1 0 0>}
      localJointOffset=<7 0 0>
     }]
    relativeSpaceBone {boneName=root}
    sphericalLimits [{limitRadius=15, drivingBone {boneName=pelvis}, sphereLocalOffset=<0 0 0>}]
   }
   n16 = Anim Dynamics - Bone: dyn_r_hair_07!(ComponentPose=@n15, Alpha=1, GravityOverride:Vector="0,0,-980", ExternalForce="0,0,0") -> (Pose) {
    angularDampingOverride=1
    angularSpringConstant=20000
    bAngularSpring=true
    bOverrideAngularDamping=true
    bUseGravityOverride=true
    bUseSphericalLimits=true
    gravityScale=0.35
    alphaBoolBlend {blendInTime=2, blendOption=Cubic}
    boundBone {boneName=dyn_r_hair_07}
    componentAppliedLinearAccClamp=<400000 400000 400000>
    componentLinearAccScale=<0.3 0.4 0.3>
    componentLinearVelScale=<0.3 0.4 3>
    physicsBodyDefinitions [{
      boundBone {boneName=dyn_r_hair_07}
      boxExtents=<25 25 25>
      constraintSetup {angularLimitsMax=<0 25 0>, angularLimitsMin=<0 -10 0>, angularTarget=<1 0 0>}
      localJointOffset=<7 0 0>
     }]
    relativeSpaceBone {boneName=root}
    sphericalLimits [{limitRadius=15, drivingBone {boneName=pelvis}, sphereLocalOffset=<0 0 0>}]
   }
   Save cached pose 'Hair_Backpack_Pose_Cache'(Pose=Component To Local(ComponentPose=@n16)) {cacheName=Hair_Backpack_Pose_Cache}
   n17 = Anim Dynamics - Bone: dyn_c_hair_01!(
    ComponentPose=Local To Component(LocalPose=Use cached pose 'CR_HairPose_Cache'())
    bAlphaBoolEnabled:bool=True
    GravityOverride:Vector="0,0,-980"
    AngularSpringConstant:float=5000
    ExternalForce="0,0,0"
   ) -> (Pose) {
    alphaInputType=Bool
    angularDampingOverride=0.9999
    angularSpringConstant=5000
    bAngularSpring=true
    bOverrideAngularDamping=true
    bUseGravityOverride=true
    gravityScale=0.35
    simulationSpace=RootRelative
    alphaBoolBlend {blendInTime=2, blendOption=Cubic}
    boundBone {boneName=dyn_c_hair_01}
    componentAppliedLinearAccClamp=<3000 3000 3000>
    componentLinearAccScale=<0.3 0.4 0.4>
    componentLinearVelScale=<0.3 0.4 0.4>
    gravityOverride {z=-980}
    physicsBodyDefinitions [{
      boundBone {boneName=dyn_c_hair_01}
      boxExtents=<20 20 20>
      constraintSetup {angularLimitsMax=<0 0 0>, angularLimitsMin=<0 0 -45>, angularTarget=<1 0 0>}
      localJointOffset=<1 0 0>
     }]
    relativeSpaceBone {boneName=root}
   }
   n18 = Anim Dynamics - Bone: dyn_l_hair_01!(ComponentPose=@n17, bAlphaBoolEnabled:bool=True, GravityOverride:Vector="0,0,-980", AngularSpringConstant:float=5000, ExternalForce="0,0,0") -> (Pose) {
    alphaInputType=Bool
    angularDampingOverride=0.9999
    angularSpringConstant=5000
    bAngularSpring=true
    bOverrideAngularDamping=true
    bUseGravityOverride=true
    gravityScale=0.35
    simulationSpace=RootRelative
    alphaBoolBlend {blendInTime=2, blendOption=Cubic}
    boundBone {boneName=dyn_l_hair_01}
    componentAppliedLinearAccClamp=<3000 3000 3000>
    componentLinearAccScale=<0.3 0.4 0.4>
    componentLinearVelScale=<0.3 0.4 0.4>
    gravityOverride {z=-980}
    physicsBodyDefinitions [{
      boundBone {boneName=dyn_l_hair_01}
      boxExtents=<20 20 20>
      constraintSetup {angularLimitsMax=<0 0 0>, angularLimitsMin=<0 0 -45>, angularTarget=<1 0 0>}
      localJointOffset=<1 0 0>
     }]
    relativeSpaceBone {boneName=root}
   }
   n19 = Anim Dynamics - Bone: dyn_r_hair_01!(ComponentPose=@n18, bAlphaBoolEnabled:bool=True, GravityOverride:Vector="0,0,-980", AngularSpringConstant:float=5000, ExternalForce="0,0,0") -> (Pose) {
    alphaInputType=Bool
    angularDampingOverride=0.9999
    angularSpringConstant=5000
    bAngularSpring=true
    bOverrideAngularDamping=true
    bUseGravityOverride=true
    gravityScale=0.35
    simulationSpace=RootRelative
    alphaBoolBlend {blendInTime=2, blendOption=Cubic}
    boundBone {boneName=dyn_r_hair_01}
    componentAppliedLinearAccClamp=<3000 3000 3000>
    componentLinearAccScale=<0.3 0.4 0.4>
    componentLinearVelScale=<0.3 0.4 0.4>
    gravityOverride {z=-980}
    physicsBodyDefinitions [{
      boundBone {boneName=dyn_r_hair_01}
      boxExtents=<20 20 20>
      constraintSetup {angularLimitsMax=<0 0 0>, angularLimitsMin=<0 0 -45>, angularTarget=<1 0 0>}
      localJointOffset=<1 0 0>
     }]
    relativeSpaceBone {boneName=root}
   }
   n20 = Anim Dynamics - Bone: dyn_c_hair_02!(ComponentPose=@n19, bAlphaBoolEnabled:bool=True, GravityOverride:Vector="0,0,-980", AngularSpringConstant:float=5000, ExternalForce="0,0,0") -> (Pose) {
    alphaInputType=Bool
    angularDampingOverride=0.9999
    angularSpringConstant=5000
    bAngularSpring=true
    bOverrideAngularDamping=true
    bUseGravityOverride=true
    gravityScale=0.35
    simulationSpace=RootRelative
    alphaBoolBlend {blendInTime=2, blendOption=Cubic}
    boundBone {boneName=dyn_c_hair_02}
    componentAppliedLinearAccClamp=<3000 3000 3000>
    componentLinearAccScale=<0.3 0.4 0.4>
    componentLinearVelScale=<0.3 0.4 0.4>
    gravityOverride {z=-980}
    physicsBodyDefinitions [{
      boundBone {boneName=dyn_c_hair_02}
      boxExtents=<20 20 20>
      constraintSetup {angularLimitsMax=<0 0 0>, angularLimitsMin=<0 0 -45>, angularTarget=<1 0 0>}
      localJointOffset=<1.25 0 0>
     }]
    relativeSpaceBone {boneName=root}
   }
   n21 = Anim Dynamics - Bone: dyn_l_hair_02!(ComponentPose=@n20, bAlphaBoolEnabled:bool=True, GravityOverride:Vector="0,0,-980", AngularSpringConstant:float=5000, ExternalForce="0,0,0") -> (Pose) {
    alphaInputType=Bool
    angularDampingOverride=0.9999
    angularSpringConstant=5000
    bAngularSpring=true
    bOverrideAngularDamping=true
    bUseGravityOverride=true
    gravityScale=0.35
    simulationSpace=RootRelative
    alphaBoolBlend {blendInTime=2, blendOption=Cubic}
    boundBone {boneName=dyn_l_hair_02}
    componentAppliedLinearAccClamp=<3000 3000 3000>
    componentLinearAccScale=<0.3 0.4 0.4>
    componentLinearVelScale=<0.3 0.4 0.4>
    gravityOverride {z=-980}
    physicsBodyDefinitions [{
      boundBone {boneName=dyn_l_hair_02}
      boxExtents=<20 20 20>
      constraintSetup {angularLimitsMax=<0 0 0>, angularLimitsMin=<0 0 -45>, angularTarget=<1 0 0>}
      localJointOffset=<1.25 0 0>
     }]
    relativeSpaceBone {boneName=root}
   }
   n22 = Anim Dynamics - Bone: dyn_r_hair_02!(ComponentPose=@n21, bAlphaBoolEnabled:bool=True, GravityOverride:Vector="0,0,-980", AngularSpringConstant:float=5000, ExternalForce="0,0,0") -> (Pose) {
    alphaInputType=Bool
    angularDampingOverride=0.9999
    angularSpringConstant=5000
    bAngularSpring=true
    bOverrideAngularDamping=true
    bUseGravityOverride=true
    gravityScale=0.35
    simulationSpace=RootRelative
    alphaBoolBlend {blendInTime=2, blendOption=Cubic}
    boundBone {boneName=dyn_r_hair_02}
    componentAppliedLinearAccClamp=<3000 3000 3000>
    componentLinearAccScale=<0.3 0.4 0.4>
    componentLinearVelScale=<0.3 0.4 0.4>
    gravityOverride {z=-980}
    physicsBodyDefinitions [{
      boundBone {boneName=dyn_r_hair_02}
      boxExtents=<20 20 20>
      constraintSetup {angularLimitsMax=<0 0 0>, angularLimitsMin=<0 0 -45>, angularTarget=<1 0 0>}
      localJointOffset=<1.25 0 0>
     }]
    relativeSpaceBone {boneName=root}
   }
   n23 = Anim Dynamics - Bone: dyn_c_hair_03!(ComponentPose=@n22, bAlphaBoolEnabled:bool=True, GravityOverride:Vector="0,0,-980", AngularSpringConstant:float=5000, ExternalForce="0,0,0") -> (Pose) {
    alphaInputType=Bool
    angularDampingOverride=0.9999
    angularSpringConstant=5000
    bAngularSpring=true
    bOverrideAngularDamping=true
    bUseGravityOverride=true
    gravityScale=0.35
    simulationSpace=RootRelative
    alphaBoolBlend {blendInTime=2, blendOption=Cubic}
    boundBone {boneName=dyn_c_hair_03}
    componentAppliedLinearAccClamp=<3000 3000 3000>
    componentLinearAccScale=<0.3 0.4 0.4>
    componentLinearVelScale=<0.3 0.4 0.4>
    gravityOverride {z=-980}
    physicsBodyDefinitions [{
      boundBone {boneName=dyn_c_hair_03}
      boxExtents=<20 20 20>
      constraintSetup {angularLimitsMax=<0 0 3>, angularLimitsMin=<0 0 -45>, angularTarget=<1 0 0>}
      localJointOffset=<1.25 0 0>
     }]
    relativeSpaceBone {boneName=root}
   }
   n24 = Anim Dynamics - Bone: dyn_l_hair_03!(ComponentPose=@n23, bAlphaBoolEnabled:bool=True, GravityOverride:Vector="0,0,-980", AngularSpringConstant:float=5000, ExternalForce="0,0,0") -> (Pose) {
    alphaInputType=Bool
    angularDampingOverride=0.9999
    angularSpringConstant=5000
    bAngularSpring=true
    bOverrideAngularDamping=true
    bUseGravityOverride=true
    gravityScale=0.35
    simulationSpace=RootRelative
    alphaBoolBlend {blendInTime=2, blendOption=Cubic}
    boundBone {boneName=dyn_l_hair_03}
    componentAppliedLinearAccClamp=<3000 3000 3000>
    componentLinearAccScale=<0.3 0.4 0.4>
    componentLinearVelScale=<0.3 0.4 0.4>
    gravityOverride {z=-980}
    physicsBodyDefinitions [{
      boundBone {boneName=dyn_l_hair_03}
      boxExtents=<20 20 20>
      constraintSetup {angularLimitsMax=<0 0 3>, angularLimitsMin=<0 0 -45>, angularTarget=<1 0 0>}
      localJointOffset=<1.25 0 0>
     }]
    relativeSpaceBone {boneName=root}
   }
   n25 = Anim Dynamics - Bone: dyn_r_hair_03!(ComponentPose=@n24, bAlphaBoolEnabled:bool=True, GravityOverride:Vector="0,0,-980", AngularSpringConstant:float=5000, ExternalForce="0,0,0") -> (Pose) {
    alphaInputType=Bool
    angularDampingOverride=0.9999
    angularSpringConstant=5000
    bAngularSpring=true
    bOverrideAngularDamping=true
    bUseGravityOverride=true
    gravityScale=0.35
    simulationSpace=RootRelative
    alphaBoolBlend {blendInTime=2, blendOption=Cubic}
    boundBone {boneName=dyn_r_hair_03}
    componentAppliedLinearAccClamp=<3000 3000 3000>
    componentLinearAccScale=<0.3 0.4 0.4>
    componentLinearVelScale=<0.3 0.4 0.4>
    gravityOverride {z=-980}
    physicsBodyDefinitions [{
      boundBone {boneName=dyn_r_hair_03}
      boxExtents=<20 20 20>
      constraintSetup {angularLimitsMax=<0 0 3>, angularLimitsMin=<0 0 -45>, angularTarget=<1 0 0>}
      localJointOffset=<1.25 0 0>
     }]
    relativeSpaceBone {boneName=root}
   }
   n26 = Anim Dynamics - Bone: dyn_c_hair_04!(ComponentPose=@n25, bAlphaBoolEnabled:bool=True, GravityOverride:Vector="0,0,-980", AngularSpringConstant:float=5000, ExternalForce="0,0,0") -> (Pose) {
    alphaInputType=Bool
    angularDampingOverride=0.9999
    angularSpringConstant=5000
    bAngularSpring=true
    bOverrideAngularDamping=true
    bUseGravityOverride=true
    gravityScale=0.35
    simulationSpace=RootRelative
    alphaBoolBlend {blendInTime=2, blendOption=Cubic}
    boundBone {boneName=dyn_c_hair_04}
    componentAppliedLinearAccClamp=<3000 3000 3000>
    componentLinearAccScale=<0.3 0.4 0.4>
    componentLinearVelScale=<0.3 0.4 0.4>
    gravityOverride {z=-980}
    physicsBodyDefinitions [{
      boundBone {boneName=dyn_c_hair_04}
      boxExtents=<20 20 20>
      constraintSetup {angularLimitsMax=<0 6 10>, angularLimitsMin=<0 -6 -35>, angularTarget=<1 0 0>}
      localJointOffset=<1.25 0 0>
     }]
    relativeSpaceBone {boneName=root}
   }
   n27 = Anim Dynamics - Bone: dyn_l_hair_04!(ComponentPose=@n26, bAlphaBoolEnabled:bool=True, GravityOverride:Vector="0,0,-980", AngularSpringConstant:float=5000, ExternalForce="0,0,0") -> (Pose) {
    alphaInputType=Bool
    angularDampingOverride=0.9999
    angularSpringConstant=5000
    bAngularSpring=true
    bOverrideAngularDamping=true
    bUseGravityOverride=true
    gravityScale=0.35
    simulationSpace=RootRelative
    alphaBoolBlend {blendInTime=2, blendOption=Cubic}
    boundBone {boneName=dyn_l_hair_04}
    componentAppliedLinearAccClamp=<3000 3000 3000>
    componentLinearAccScale=<0.3 0.4 0.4>
    componentLinearVelScale=<0.3 0.4 0.4>
    gravityOverride {z=-980}
    physicsBodyDefinitions [{
      boundBone {boneName=dyn_l_hair_04}
      boxExtents=<20 20 20>
      constraintSetup {angularLimitsMax=<0 6 10>, angularLimitsMin=<0 -6 -35>, angularTarget=<1 0 0>}
      localJointOffset=<1.25 0 0>
     }]
    relativeSpaceBone {boneName=root}
   }
   n28 = Anim Dynamics - Bone: dyn_r_hair_04!(ComponentPose=@n27, bAlphaBoolEnabled:bool=True, GravityOverride:Vector="0,0,-980", AngularSpringConstant:float=5000, ExternalForce="0,0,0") -> (Pose) {
    alphaInputType=Bool
    angularDampingOverride=0.9999
    angularSpringConstant=5000
    bAngularSpring=true
    bOverrideAngularDamping=true
    bUseGravityOverride=true
    gravityScale=0.35
    simulationSpace=RootRelative
    alphaBoolBlend {blendInTime=2, blendOption=Cubic}
    boundBone {boneName=dyn_r_hair_04}
    componentAppliedLinearAccClamp=<3000 3000 3000>
    componentLinearAccScale=<0.3 0.4 0.4>
    componentLinearVelScale=<0.3 0.4 0.4>
    gravityOverride {z=-980}
    physicsBodyDefinitions [{
      boundBone {boneName=dyn_r_hair_04}
      boxExtents=<20 20 20>
      constraintSetup {angularLimitsMax=<0 6 10>, angularLimitsMin=<0 -6 -35>, angularTarget=<1 0 0>}
      localJointOffset=<1.25 0 0>
     }]
    relativeSpaceBone {boneName=root}
   }
   n29 = Anim Dynamics - Bone: dyn_c_hair_05!(ComponentPose=@n28, bAlphaBoolEnabled:bool=True, GravityOverride:Vector="0,0,-980", AngularSpringConstant:float=5000, ExternalForce="0,0,0") -> (Pose) {
    alphaInputType=Bool
    angularDampingOverride=0.9999
    angularSpringConstant=5000
    bAngularSpring=true
    bOverrideAngularDamping=true
    bUseGravityOverride=true
    gravityScale=0.35
    simulationSpace=RootRelative
    alphaBoolBlend {blendInTime=2, blendOption=Cubic}
    boundBone {boneName=dyn_c_hair_05}
    componentAppliedLinearAccClamp=<3000 3000 3000>
    componentLinearAccScale=<0.3 0.4 0.3>
    componentLinearVelScale=<0.3 0.4 0.3>
    gravityOverride {z=-980}
    physicsBodyDefinitions [{
      boundBone {boneName=dyn_c_hair_05}
      boxExtents=<20 20 20>
      constraintSetup {angularLimitsMax=<0 10 0>, angularLimitsMin=<0 -10 -10>, angularTarget=<1 0 0>}
      localJointOffset=<1.1 0 0>
     }]
    relativeSpaceBone {boneName=root}
   }
   n30 = Anim Dynamics - Bone: dyn_l_hair_05!(ComponentPose=@n29, bAlphaBoolEnabled:bool=True, GravityOverride:Vector="0,0,-980", AngularSpringConstant:float=5000, ExternalForce="0,0,0") -> (Pose) {
    alphaInputType=Bool
    angularDampingOverride=0.9999
    angularSpringConstant=5000
    bAngularSpring=true
    bOverrideAngularDamping=true
    bUseGravityOverride=true
    gravityScale=0.35
    simulationSpace=RootRelative
    alphaBoolBlend {blendInTime=2, blendOption=Cubic}
    boundBone {boneName=dyn_l_hair_05}
    componentAppliedLinearAccClamp=<3000 3000 3000>
    componentLinearAccScale=<0.3 0.4 0.3>
    componentLinearVelScale=<0.3 0.4 0.3>
    gravityOverride {z=-980}
    physicsBodyDefinitions [{
      boundBone {boneName=dyn_l_hair_05}
      boxExtents=<20 20 20>
      constraintSetup {angularLimitsMax=<0 10 0>, angularLimitsMin=<0 -10 -10>, angularTarget=<1 0 0>}
      localJointOffset=<1.1 0 0>
     }]
    relativeSpaceBone {boneName=root}
   }
   n31 = Anim Dynamics - Bone: dyn_r_hair_05!(ComponentPose=@n30, bAlphaBoolEnabled:bool=True, GravityOverride:Vector="0,0,-980", AngularSpringConstant:float=5000, ExternalForce="0,0,0") -> (Pose) {
    alphaInputType=Bool
    angularDampingOverride=0.9999
    angularSpringConstant=5000
    bAngularSpring=true
    bOverrideAngularDamping=true
    bUseGravityOverride=true
    gravityScale=0.35
    simulationSpace=RootRelative
    alphaBoolBlend {blendInTime=2, blendOption=Cubic}
    boundBone {boneName=dyn_r_hair_05}
    componentAppliedLinearAccClamp=<3000 3000 3000>
    componentLinearAccScale=<0.3 0.4 0.3>
    componentLinearVelScale=<0.3 0.4 0.3>
    gravityOverride {z=-980}
    physicsBodyDefinitions [{
      boundBone {boneName=dyn_r_hair_05}
      boxExtents=<20 20 20>
      constraintSetup {angularLimitsMax=<0 10 0>, angularLimitsMin=<0 -10 -10>, angularTarget=<1 0 0>}
      localJointOffset=<1.1 0 0>
     }]
    relativeSpaceBone {boneName=root}
   }
   n32 = Anim Dynamics - Bone: dyn_c_hair_06!(ComponentPose=@n31, bAlphaBoolEnabled:bool=True, GravityOverride="0,0,-980", AngularSpringConstant:float=5000, ExternalForce="0,0,0") -> (Pose) {
    alphaInputType=Bool
    angularDampingOverride=0.9999
    angularSpringConstant=5000
    bAngularSpring=true
    bOverrideAngularDamping=true
    bUseGravityOverride=true
    bUseSphericalLimits=true
    gravityScale=0.35
    simulationSpace=RootRelative
    alphaBoolBlend {blendInTime=2, blendOption=Cubic}
    boundBone {boneName=dyn_c_hair_06}
    componentAppliedLinearAccClamp=<3000 3000 3000>
    componentLinearAccScale=<0.3 0.4 0.3>
    componentLinearVelScale=<0.3 0.4 0.3>
    gravityOverride {z=-980}
    physicsBodyDefinitions [{
      boundBone {boneName=dyn_c_hair_06}
      boxExtents=<19 19 19>
      constraintSetup {angularLimitsMax=<0 10 0>, angularLimitsMin=<0 -10 -10>, angularTarget=<1 0 0>}
      localJointOffset=<1.2 0 0>
     }]
    relativeSpaceBone {boneName=root}
    sphericalLimits [{limitRadius=15, drivingBone {boneName=pelvis}, sphereLocalOffset=<0 0 0>}]
   }
   n33 = Anim Dynamics - Bone: dyn_l_hair_06!(ComponentPose=@n32, bAlphaBoolEnabled:bool=True, GravityOverride="0,0,-980", AngularSpringConstant:float=5000, ExternalForce="0,0,0") -> (Pose) {
    alphaInputType=Bool
    angularDampingOverride=0.9999
    angularSpringConstant=5000
    bAngularSpring=true
    bOverrideAngularDamping=true
    bUseGravityOverride=true
    bUseSphericalLimits=true
    gravityScale=0.35
    simulationSpace=RootRelative
    alphaBoolBlend {blendInTime=2, blendOption=Cubic}
    boundBone {boneName=dyn_l_hair_06}
    componentAppliedLinearAccClamp=<3000 3000 3000>
    componentLinearAccScale=<0.3 0.4 0.3>
    componentLinearVelScale=<0.3 0.4 0.3>
    gravityOverride {z=-980}
    physicsBodyDefinitions [{
      boundBone {boneName=dyn_l_hair_06}
      boxExtents=<19 19 19>
      constraintSetup {angularLimitsMax=<0 10 0>, angularLimitsMin=<0 -10 -10>, angularTarget=<1 0 0>}
      localJointOffset=<1.2 0 0>
     }]
    relativeSpaceBone {boneName=root}
    sphericalLimits [{limitRadius=15, drivingBone {boneName=pelvis}, sphereLocalOffset=<0 0 0>}]
   }
   n34 = Anim Dynamics - Bone: dyn_r_hair_06!(ComponentPose=@n33, bAlphaBoolEnabled:bool=True, GravityOverride="0,0,-980", AngularSpringConstant:float=5000, ExternalForce="0,0,0") -> (Pose) {
    alphaInputType=Bool
    angularDampingOverride=0.9999
    angularSpringConstant=5000
    bAngularSpring=true
    bOverrideAngularDamping=true
    bUseGravityOverride=true
    bUseSphericalLimits=true
    gravityScale=0.35
    simulationSpace=RootRelative
    alphaBoolBlend {blendInTime=2, blendOption=Cubic}
    boundBone {boneName=dyn_r_hair_06}
    componentAppliedLinearAccClamp=<3000 3000 3000>
    componentLinearAccScale=<0.3 0.4 0.3>
    componentLinearVelScale=<0.3 0.4 0.3>
    gravityOverride {z=-980}
    physicsBodyDefinitions [{
      boundBone {boneName=dyn_r_hair_06}
      boxExtents=<19 19 19>
      constraintSetup {angularLimitsMax=<0 10 0>, angularLimitsMin=<0 -10 -10>, angularTarget=<1 0 0>}
      localJointOffset=<1.2 0 0>
     }]
    relativeSpaceBone {boneName=root}
    sphericalLimits [{limitRadius=15, drivingBone {boneName=pelvis}, sphereLocalOffset=<0 0 0>}]
   }
   n35 = Anim Dynamics - Bone: dyn_c_hair_07!(ComponentPose=@n34, bAlphaBoolEnabled:bool=True, GravityOverride="0,0,-980", AngularSpringConstant:float=5000, ExternalForce="0,0,0") -> (Pose) {
    alphaInputType=Bool
    angularDampingOverride=0.9999
    angularSpringConstant=5000
    bAngularSpring=true
    bOverrideAngularDamping=true
    bUseGravityOverride=true
    bUseSphericalLimits=true
    gravityScale=0.35
    simulationSpace=RootRelative
    alphaBoolBlend {blendInTime=2, blendOption=Cubic}
    boundBone {boneName=dyn_c_hair_07}
    componentAppliedLinearAccClamp=<3000 3000 3000>
    componentLinearAccScale=<0.3 0.4 0.3>
    componentLinearVelScale=<0.3 0.4 0.3>
    gravityOverride {z=-980}
    physicsBodyDefinitions [{
      boundBone {boneName=dyn_c_hair_07}
      boxExtents=<18 18 18>
      constraintSetup {angularLimitsMax=<0 10 0>, angularLimitsMin=<0 -10 -10>, angularTarget=<1 0 0>}
      localJointOffset=<1.3 0 0>
     }]
    relativeSpaceBone {boneName=root}
    sphericalLimits [{limitRadius=15, drivingBone {boneName=pelvis}, sphereLocalOffset=<0 0 0>}]
   }
   n36 = Anim Dynamics - Bone: dyn_l_hair_07!(ComponentPose=@n35, bAlphaBoolEnabled:bool=True, GravityOverride="0,0,-980", AngularSpringConstant:float=5000, ExternalForce="0,0,0") -> (Pose) {
    alphaInputType=Bool
    angularDampingOverride=0.9999
    angularSpringConstant=5000
    bAngularSpring=true
    bOverrideAngularDamping=true
    bUseGravityOverride=true
    bUseSphericalLimits=true
    gravityScale=0.35
    simulationSpace=RootRelative
    alphaBoolBlend {blendInTime=2, blendOption=Cubic}
    boundBone {boneName=dyn_l_hair_07}
    componentAppliedLinearAccClamp=<3000 3000 3000>
    componentLinearAccScale=<0.3 0.4 0.3>
    componentLinearVelScale=<0.3 0.4 0.3>
    gravityOverride {z=-980}
    physicsBodyDefinitions [{
      boundBone {boneName=dyn_l_hair_07}
      boxExtents=<18 18 18>
      constraintSetup {angularLimitsMax=<0 10 0>, angularLimitsMin=<0 -10 -10>, angularTarget=<1 0 0>}
      localJointOffset=<1.3 0 0>
     }]
    relativeSpaceBone {boneName=root}
    sphericalLimits [{limitRadius=15, drivingBone {boneName=pelvis}, sphereLocalOffset=<0 0 0>}]
   }
   n37 = Anim Dynamics - Bone: dyn_r_hair_07!(ComponentPose=@n36, bAlphaBoolEnabled:bool=True, GravityOverride="0,0,-980", AngularSpringConstant:float=5000, ExternalForce="0,0,0") -> (Pose) {
    alphaInputType=Bool
    angularDampingOverride=0.9999
    angularSpringConstant=5000
    bAngularSpring=true
    bOverrideAngularDamping=true
    bUseGravityOverride=true
    bUseSphericalLimits=true
    gravityScale=0.35
    simulationSpace=RootRelative
    alphaBoolBlend {blendInTime=2, blendOption=Cubic}
    boundBone {boneName=dyn_r_hair_07}
    componentAppliedLinearAccClamp=<3000 3000 3000>
    componentLinearAccScale=<0.3 0.4 0.3>
    componentLinearVelScale=<0.3 0.4 0.3>
    gravityOverride {z=-980}
    physicsBodyDefinitions [{
      boundBone {boneName=dyn_r_hair_07}
      boxExtents=<18 18 18>
      constraintSetup {angularLimitsMax=<0 10 0>, angularLimitsMin=<0 -10 -10>, angularTarget=<1 0 0>}
      localJointOffset=<1.3 0 0>
     }]
    relativeSpaceBone {boneName=root}
    sphericalLimits [{limitRadius=15, drivingBone {boneName=pelvis}, sphereLocalOffset=<0 0 0>}]
   }
   n38 = RigidBody!(
    ComponentPose=@n37
    Alpha=Get RBAN_Alpha() -> (RBAN_Alpha:double)
    OverrideWorldGravity=Get Gravity Override()
    ExternalForce
    ComponentLinearAccScale:Vector="0.3,0.4,0.4"
    ComponentLinearVelScale:Vector="0.3,0.4,0.4"
    SimSpaceSettings:Sim Space Settings Structure=(WorldAlpha=0.2,VelocityScaleZ=1,DampingAlpha=1,MaxLinearVelocity=150,MaxAngularVelocity=10,MaxLinearAcceleration=50,MaxAngularAcceleration=5,ExternalLinearDragV=(X=0,Y=0,Z=0),ExternalLinearVelocity=(X=0,Y=0,Z=0),ExternalAngularVelocity=(X=0,Y=0,Z=0))
   ) -> (Pose) {
    bAllowPoseScalePropagation=false
    bClampLinearTranslationLimitToRefPose=true
    bOverrideWorldGravity=true
    overridePhysicsAsset=&F_MED_MissusMind_Suit_FaceAcc_Physics
    simulationSpace=BaseBoneSpace
    alphaBoolBlend {blendOption=Linear}
    baseBoneRef {boneName=head}
    componentAppliedLinearAccClamp=<1000 1000 1000>
    componentLinearAccScale=<0.3 0.4 0.4>
    componentLinearVelScale=<0.3 0.4 0.4>
    simSpaceSettings {maxAngularAcceleration=5, maxAngularVelocity=10, maxLinearAcceleration=50, maxLinearVelocity=150, worldAlpha=0.2}
   }
   Save cached pose 'Dyn_Hair_Cache'(Pose=Component To Local(ComponentPose=@n38)) {cacheName=Dyn_Hair_Cache}
   n39 = Blend Poses by bool(
    BlendPose_0=Use cached pose 'Hair_Backpack_Pose_Cache'()
    BlendPose_1=Blend Poses by bool(BlendPose_0=Use cached pose 'Dyn_Hair_Cache'(), BlendPose_1=Use cached pose 'CR_HairPose_Cache'()) {blendType=Linear}
   ) {blendType=Linear}
   n40 = Use cached pose 'Hair_Backpack_Pose_Cache'()
   Save cached pose 'Blend_Hair_Pose_Cache'(Pose=Blend Poses by bool(
     BlendPose_0=Use cached pose 'Hair_Backpack_Pose_Cache'()
     BlendPose_1=Blend Poses by bool(BlendPose_0=@n40, BlendPose_1=@n39) {blendType=Linear}
    ) {blendType=Linear}) {cacheName=Blend_Hair_Pose_Cache}
  ]}
 EventGraph {nodes [
   Event BlueprintUpdateAnimation()
   > n1 = Sequence!(execute) -> (then_0, then_1, then_2:Exec)
   > n2 = Gravity Override State Function DEPRECATED(
    CharacterPartAnimInstance=Self-Reference()
    GameTimeInSeconds=GetGameTimeInSeconds()
    GravityOverrideParams=Get Gravity Override Params()
    RBAN Params=Get RBAN Params()
   )
   > n3 = Set Gravity Override(Gravity Override=@n2)
   > n4 = Set Gravity Override_AnimDynamics(Gravity Override_AnimDynamics:Vector=@n3) -> (Output_Get:Vector)
   > Set RBAN Settings Out(RBAN Settings Out:RBANSetting Struct Structure=@n2.RBAN SettingsOut) -> (Output_Get:RBANSetting Struct Structure)
   n5 = Get bIsSkyDiving()
   n6 = Get bIsJumping()
   n7 = Get bIsFalling()
   > n8 = Branch(Condition=AND Boolean(A=OR Boolean(A=@n6, B=@n7), B=NOT Boolean(A=@n5)))
   n9 = Remap Float(Input=Get PawnSpeed2D(), OldRange=(X=0,Y=550), NewRange=(X=0,Y=1), Clamp=true) -> (Output=~)
   n10 = "Clamp (Float)"(Value=Get PawnVelocityZ(), Min=-980, Max=0)
   n11 = float + float(A=Lerp(A=1960, B=float * float(A=@n10, B=-2) {is=title K2Node_PromotableOperator}, Alpha=@n9), B=Lerp(A=0, B=800, Alpha=@n9))
   > Set Gravity Override_AnimDynamics(Gravity Override_AnimDynamics=vector + vector(A=@n4, B=MakeVector(Y=0, Z=@n11)))
   Set Gravity Override_AnimDynamics(execute=@n8.else, Gravity Override_AnimDynamics=@n4)
   n12 = Branch(execute=@n1.then_1, Condition=Get bIsSkyDiving())
   > Set RBAN_Alpha(RBAN_Alpha:double=0.7) -> (Output_Get:double)
   > Set AngularSpringConstant(AngularSpringConstant:double=10000) -> (Output_Get:double)
   Set RBAN_Alpha(execute=@n12.else, RBAN_Alpha=1)
   > Set AngularSpringConstant(AngularSpringConstant=6000)
   Set hasMatchingGlider(
    execute=@n1.then_2
    hasMatchingGlider:bool=GliderItemDefinitionHasAnyMetaTag!(Tags:Array of Gameplay Tag Structures=Get Tags() -> (Tags:Array of Gameplay Tag Structures), bRequiresOpened:bool=true) -> (ReturnValue:bool)
   ) -> (Output_Get:bool)
   Event OnCharacterCustomizationLoaded()
   n13 = CheckForBackbling()
   > Set isWearingPack(isWearingPack:bool=OR Boolean!(A=@n13, B=@n13.bIsBackblingACape, C:bool=@n13.bIsBackblingLong) -> (ReturnValue)) -> (Output_Get:bool)
   > Set isWearingCape(isWearingCape=@n13.bIsBackblingLong)
  ]}
]
variables [
 AngularSpringConstant {type=number, default=7500}
 animDynamicsEnabled {type=boolean, default=true}
 Gravity Override {type=Vector}
 Gravity Override Params {type=GravityOverrideParamsStruct, default {onGround_Moving {gravityOverride=<0 0 -800>}}}
 Gravity Override_AnimDynamics {type=Vector}
 hasMatchingGlider {type=boolean}
 isSkydiveGliding {type=boolean}
 isWearingCape {type=boolean}
 isWearingPack {type=boolean}
 RBAN Params {
  type=RBANParamsStruct
  default {
   dBNO {
    componentLinearAccScale=<0.3 0.4 0.4>
    componentLinearVelScale=<0.4 0.5 0.5>
    simSpaceSettings {
     maxAngularAcceleration=5
     maxAngularVelocity=10
     maxLinearAcceleration=50
     maxLinearVelocity=150
     velocityScaleZ=1
     worldAlpha=0.2
     externalLinearDragV=<0 0 0>
    }
   }
   dBNO_Carried {
    componentLinearAccScale=<0.3 0.4 0.4>
    componentLinearVelScale=<0.4 0.5 0.5>
    simSpaceSettings {
     maxAngularAcceleration=5
     maxAngularVelocity=10
     maxLinearAcceleration=50
     maxLinearVelocity=150
     velocityScaleZ=1
     worldAlpha=0.2
     externalLinearDragV=<0 0 0>
    }
   }
   emoteOrMelee {
    componentLinearAccScale=<0.3 0.4 0.4>
    componentLinearVelScale=<0.4 0.5 0.5>
    simSpaceSettings {
     maxAngularAcceleration=5
     maxAngularVelocity=10
     maxLinearAcceleration=50
     maxLinearVelocity=150
     velocityScaleZ=1
     worldAlpha=0.2
     externalLinearDragV=<0 0 0>
    }
   }
   front_End {
    componentLinearAccScale=<0.3 0.4 0.4>
    componentLinearVelScale=<0.4 0.5 0.5>
    simSpaceSettings {
     maxAngularAcceleration=5
     maxAngularVelocity=10
     maxLinearAcceleration=50
     maxLinearVelocity=150
     velocityScaleZ=1
     worldAlpha=0.2
     externalLinearDragV=<0 0 0>
    }
   }
   onGround_Crouch_Moving {
    componentLinearAccScale=<0.3 0.4 0.4>
    componentLinearVelScale=<0.4 0.5 0.5>
    simSpaceSettings {
     maxAngularAcceleration=5
     maxAngularVelocity=10
     maxLinearAcceleration=50
     maxLinearVelocity=150
     velocityScaleZ=1
     worldAlpha=0.2
     externalLinearDragV=<0 0 0>
    }
   }
   onGround_Falling {
    componentLinearAccScale=<0.3 0.4 0.4>
    componentLinearVelScale=<0.4 0.5 0.5>
    simSpaceSettings {
     maxAngularAcceleration=5
     maxAngularVelocity=10
     maxLinearAcceleration=50
     maxLinearVelocity=150
     velocityScaleZ=1
     worldAlpha=0.2
     externalLinearDragV=<0 0 0>
    }
   }
   onGround_Moving {
    componentLinearAccScale=<0.5 0.4 0.4>
    componentLinearVelScale=<0.6 0.5 0.5>
    simSpaceSettings {
     maxAngularAcceleration=5
     maxAngularVelocity=10
     maxLinearAcceleration=50
     maxLinearVelocity=150
     velocityScaleZ=1
     worldAlpha=0.2
     externalLinearDragV=<0 0 0>
    }
   }
   onGround_Standing {
    componentLinearAccScale=<0.3 0.4 0.4>
    componentLinearVelScale=<0.4 0.5 0.5>
    simSpaceSettings {
     maxAngularAcceleration=5
     maxAngularVelocity=10
     maxLinearAcceleration=50
     maxLinearVelocity=150
     velocityScaleZ=1
     worldAlpha=0.2
     externalLinearDragV=<0 0 0>
    }
   }
   rail_Grinding {simSpaceSettings {
     maxAngularAcceleration=1000
     maxAngularVelocity=3
     maxLinearAcceleration=300
     maxLinearVelocity=700
     velocityScaleZ=0
     worldAlpha=0.1
     externalLinearDragV=<0 0.5 0>
    }}
   riding_Moving {
    componentLinearAccScale=<0.3 0.4 0.4>
    componentLinearVelScale=<0.4 0.5 0.5>
    simSpaceSettings {
     maxAngularAcceleration=5
     maxAngularVelocity=10
     maxLinearAcceleration=50
     maxLinearVelocity=150
     velocityScaleZ=1
     worldAlpha=0.2
     externalLinearDragV=<0 0 0>
    }
   }
   riding_Moving_Targeting {
    componentLinearAccScale=<0.3 0.4 0.4>
    componentLinearVelScale=<0.4 0.5 0.5>
    simSpaceSettings {
     maxAngularAcceleration=5
     maxAngularVelocity=10
     maxLinearAcceleration=50
     maxLinearVelocity=150
     velocityScaleZ=1
     worldAlpha=0.2
     externalLinearDragV=<0 0 0>
    }
   }
   riding_Standing {
    componentLinearAccScale=<0.3 0.4 0.4>
    componentLinearVelScale=<0.4 0.5 0.5>
    simSpaceSettings {
     maxAngularAcceleration=5
     maxAngularVelocity=10
     maxLinearAcceleration=50
     maxLinearVelocity=150
     velocityScaleZ=1
     worldAlpha=0.2
     externalLinearDragV=<0 0 0>
    }
   }
   skydive_DivingDown {
    componentLinearAccScale=<0.3 0.4 0.4>
    componentLinearVelScale=<0.4 0.5 0.5>
    simSpaceSettings {
     maxAngularAcceleration=5
     maxAngularVelocity=10
     maxLinearAcceleration=50
     maxLinearVelocity=150
     velocityScaleZ=1
     worldAlpha=0.2
     externalLinearDragV=<0 0 0>
    }
   }
   skydive_FloatingDown {
    componentLinearAccScale=<0.3 0.4 0.4>
    componentLinearVelScale=<0.4 0.5 0.5>
    simSpaceSettings {
     maxAngularAcceleration=5
     maxAngularVelocity=10
     maxLinearAcceleration=50
     maxLinearVelocity=150
     velocityScaleZ=1
     worldAlpha=0.2
     externalLinearDragV=<0 0 0>
    }
   }
   skydive_GoingUp {
    componentLinearAccScale=<0.3 0.4 0.4>
    componentLinearVelScale=<0.4 0.5 0.5>
    simSpaceSettings {
     maxAngularAcceleration=5
     maxAngularVelocity=10
     maxLinearAcceleration=50
     maxLinearVelocity=150
     velocityScaleZ=1
     worldAlpha=0.2
     externalLinearDragV=<0 0 0>
    }
   }
   skydive_Parachuting {
    componentLinearAccScale=<0.3 0.4 0.4>
    componentLinearVelScale=<0.4 0.5 0.5>
    simSpaceSettings {
     maxAngularAcceleration=5
     maxAngularVelocity=10
     maxLinearAcceleration=50
     maxLinearVelocity=150
     velocityScaleZ=1
     worldAlpha=0.2
     externalLinearDragV=<0 0 0>
    }
   }
   skydive_Surfing {
    componentLinearAccScale=<0.3 0.4 0.4>
    componentLinearVelScale=<0.4 0.5 0.5>
    simSpaceSettings {
     maxAngularAcceleration=5
     maxAngularVelocity=10
     maxLinearAcceleration=50
     maxLinearVelocity=150
     velocityScaleZ=1
     worldAlpha=0.2
     externalLinearDragV=<0 0 0>
    }
   }
   swimming {
    componentLinearAccScale=<0.3 0.4 0.4>
    componentLinearVelScale=<0.4 0.5 0.5>
    simSpaceSettings {
     maxAngularAcceleration=5
     maxAngularVelocity=10
     maxLinearAcceleration=50
     maxLinearVelocity=150
     velocityScaleZ=1
     worldAlpha=0.2
     externalLinearDragV=<0 0 0>
    }
   }
   tactical_Sprint {
    componentLinearAccScale=<0.3 0.4 0.4>
    componentLinearVelScale=<0.4 0.5 0.5>
    simSpaceSettings {
     maxAngularAcceleration=5
     maxAngularVelocity=10
     maxLinearAcceleration=50
     maxLinearVelocity=150
     velocityScaleZ=1
     worldAlpha=0.2
     externalLinearDragV=<0 0 0>
    }
   }
   ziplining {simSpaceSettings {
     maxAngularAcceleration=1000
     maxAngularVelocity=3
     maxLinearAcceleration=300
     maxLinearVelocity=700
     velocityScaleZ=0
     externalLinearDragV=<0 0.5 0>
    }}
  }
 }
 RBAN Settings Out {
  type=RBANSettingStruct
  default {
   componentLinearAccScale=<0 0 0>
   componentLinearVelScale=<0 0 0>
   simSpaceSettings {
    dampingAlpha=1
    maxAngularAcceleration=1000
    maxAngularVelocity=3
    maxLinearAcceleration=300
    maxLinearVelocity=700
    velocityScaleZ=0
    worldAlpha=1
    externalAngularVelocity=<0 0 0>
    externalLinearDragV=<0 0.5 0>
    externalLinearVelocity=<0 0 0>
   }
  }
 }
 RBAN_Alpha {type=number}
 Tags {type="array<GameplayTag>", default [{tagName=Cosmetics.MetaTags.MissusMind}]}
]
